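# Parent material for RadioPaca cosmetics, layering audio-reactive equalizer, radial ray and scanline screen effects between the shared 2023 character PreFX and PostFX functions.
# Audio spectrum bands blend with idle sines to drive SDF-box cells with noise, colored through a curve atlas.
# Masks composite the effects into emissive, base color, normal and specular, plus dithered opacity mask.
# Static switches select rays, polar UVs, Slime variants, and pickaxe anisotropy.
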Material /BRCosmetics/Characters/Player/Male/Medium/Bodies/M_MED_RadioPaca/Materials/M_RadioPaca_Parent
graph {
 properties {
  bHasPixelAnimation=true
  bUseMaterialAttributes=true
  bUsedWithClothing=true
  bUsedWithInstancedStaticMeshes=true
  bUsedWithMeshDeformer=true
  bUsedWithMorphTargets=true
  bUsedWithNanite=true
  bUsedWithNiagaraMeshParticles=true
  bUsedWithSkeletalMesh=true
  pixelDepthOffsetMode=PDOM_Legacy
  shadingModel=MSM_FromMaterialExpression
  usedShadingModels=Default Lit | Subsurface | Clear Coat | Subsurface Profile | Cloth
 }
 nodes [
  ComponentMask() {r=true}
  ComponentMask() {r=true}
  n1 = NamedRerouteUsage()
  n2 = ComponentMask(None=VectorParameter() {parameterName=Param_1, defaultValue=<rgba 0.2384 0.2892 0 0>}.RGB) {g=true, r=true}
  n3 = Subtract(A=TextureCoordinate()) {constB=0.5}
  n4 = Step(Y=ComponentMask(None=Subtract(A=@n2, B=@n3)) {r=true}, X=ScalarParameter() {parameterName=Equalizer Y Step})
  n5 = Step(Y=NamedRerouteUsage()) {constX=0.1, constY=0.1}
  n6 = Multiply(A=Saturate(None=Subtract(A=@n4, B=@n5)), B=NamedRerouteUsage())
  n7 = Max(A=Multiply(A=Saturate(None=@n1), B=NamedRerouteUsage()), B=Multiply(A=@n6, B=NamedRerouteUsage()))
  n8 = NamedRerouteUsage()
  n9 = Saturate(None=StaticSwitchParameter(True=Multiply(A=@n7, B=@n8), False=@n6) {parameterName=Use Rays?})
  n10 = OneMinus(None=NamedRerouteUsage())
  n11 = SetMaterialAttributes!(
   MaterialAttributes=NamedRerouteUsage()
   Emissive Color=NamedRerouteUsage()
   Base Color=NamedRerouteUsage()
   Normal=NamedRerouteUsage()
   Specular=NamedRerouteUsage()
   Opacity=@n9
   Opacity Mask=MaterialFunctionCall(Alpha Threshold=Max(A=@n9, B=@n10)) {materialFunction=&DitherTemporalAA}.Result
   World Position Offset
  ) {attributeSetTypes [
    #1
    #2
    #3
    #4
    #5
    #6
    #7
   ]}
  n12 = TextureSampleParameter2D() {
   group=RadioPaca Pickaxe
   parameterName=Anisotropy Mask
   samplerSource=SSM_Wrap_WorldGroupSettings
   samplerType=SAMPLERTYPE_Masks
   texture=&T_BlackMask
  }
  n13 = Multiply(A=@n12.R, B=ScalarParameter() {group=RadioPaca Pickaxe, parameterName=Anisotropy Amt, sliderMax=1, sliderMin=-1})
  n14 = TextureSampleParameter2D() {
   group=RadioPaca Pickaxe
   parameterName=Tangent Map
   samplerSource=SSM_Wrap_WorldGroupSettings
   samplerType=SAMPLERTYPE_Normal
   texture=&FlatNormalBlue
  }
  n15 = Multiply(A=@n14.RGB, B=Constant3Vector() {constant=<rgba 1 1 0 0>})
  n16 = MaterialFunctionCall() {materialFunction=&MF_CharacterParent_2023_PreFX}
  MaterialFunctionCall(
   Material=StaticSwitchParameter(
    True=SetMaterialAttributes!(MaterialAttributes=@n11, Anisotropy=@n13, Tangent=@n15) {attributeSetTypes [#8, #9]}
    False=@n11
   ) {group=RadioPaca Pickaxe, parameterName=IsRadioPacaPickaxe}
   M=@n16.M
   Pre Skinned Local Position=@n16.M
   Distance Blend=@n16.M
   Fresnel=@n16.M
  ) {materialFunction=&MF_CharacterParent_2023_PostFX}
  n17 = AppendVector(A=@n16.Tech Art Mask, B=@n16.Tech Art Mask)
  n18 = Step(
   Y=ScalarParameter() {defaultValue=0.03, parameterName=Equalizer Mask - Step Amount}
   X=NamedRerouteDeclaration(None=StaticComponentMaskParameter(None=@n17) {defaultA=true, parameterName=Equalizer Mask}) {name=FX Alpha, nodeColor=<rgba 0.2422 0 1 1>}
  ) {constX=0.1, constY=0.02}
  n19 = NamedRerouteUsage()
  n20 = StaticSwitchParameter(True=Max(A=@n18, B=OneMinus(None=@n19)), False=@n18) {defaultValue=true, parameterName=Use Foot Mask?}
  NamedRerouteDeclaration(None=@n20) {name=Shirt Mask, nodeColor=<rgba 1 0 0.5156 1>}
  n21 = LinearInterpolate(
   A=VectorParameter() {parameterName=Spectrum Offset, defaultValue=<rgba -1.0545 0.8199 0 0>}.RGB
   B=VectorParameter() {parameterName=Spectrum Offset - Slime, defaultValue=<rgba -1.0545 0.8199 0 0>}.RGB
   Alpha=NamedRerouteUsage()
  )
  n22 = VectorParameter() {parameterName=Secondary - Spectrum Offset, defaultValue=<rgba -1.9599 0.4942 0 0>}
  n23 = NamedRerouteUsage()
  n24 = TextureSampleParameter2D!(UVs, Bias=ScalarParameter() {defaultValue=-1, parameterName=Bias}, Apply View MipBias) {
   mipValueMode=TMVM_MipBias
   parameterName=Param
   samplerType=SAMPLERTYPE_LinearColor
   texture=/BRCosmetics/Gadgets/Assets/VinderTech_GliderChute/Glider_RadioPaca/Textures/T_RadioPaca_Glider_FX2
  }
  n25 = Add(
   A=ComponentMask(None=LinearInterpolate(A=@n21, B=@n22.RGB, Alpha=@n23)) {g=true, r=true}
   B=StaticSwitchParameter(True=ComponentMask(None=@n24.RGB) {g=true, r=true}, False=TextureCoordinate()) {parameterName=Use Flow 2D as UV?}
  )
  n26 = MaterialFunctionCall(Vector or UVs=@n25) {materialFunction=&VectorToRadialValue}
  n27 = LinearInterpolate(
   A=@n25
   B=MaterialFunctionCall(UVs=@n25, "Rotation Angle (0-1)"=ScalarParameter() {defaultValue=0.004, parameterName=Param_3}) {materialFunction=&CustomRotator}.Rotated Values
   Alpha=NamedRerouteUsage()
  )
  NamedRerouteDeclaration(None=StaticSwitchParameter(True=@n26.Radial Coordinates, False=StaticSwitchParameter(True=@n27, False=@n25) {parameterName=Slime Variant?}) {parameterName=Use Polar Coordinates?}) {name=Tex Coord, nodeColor=<rgba 0 1 0.5313 1>}
  n28 = TextureCoordinate()
  n29 = MaterialFunctionCall(UVs=@n28, "Rotation Angle (0-1)"=ScalarParameter() {defaultValue=0.004, parameterName=Param_3}) {materialFunction=&CustomRotator}
  n30 = ComponentMask(None=VectorParameter() {parameterName=Back Mask Pos, defaultValue=<rgba 0.4 0.277 0 0>}.RGB) {g=true, r=true}
  n31 = VectorParameter() {defaultValue=<rgba 0.149 0.0879 0 0>}
  n32 = Constant()
  n33 = Step(
   Y=MaterialFunctionCall!(UVs=Add(A=@n29.Rotated Values, B=@n30), Size=@n31.RGB, CornerRoundness, KeepAspectRatio, PixelSize) {materialFunction=/FortUILibrary/MaterialFunctions/Shapes/SDFs/MF_SDF_Box}.SDF
   X=@n32
  )
  n34 = NamedRerouteUsage()
  NamedRerouteDeclaration(None=Multiply(A=LinearInterpolate(A=@n33, Alpha=@n34) {constB=0}, B=ScalarParameter() {defaultValue=1, parameterName=Back Area - UV Mask})) {name=Back Step, nodeColor=<rgba 0 0.4609 1 1>}
  n35 = TextureSample() {
   samplerType=SAMPLERTYPE_LinearGrayscale
   texture=/BRCosmetics/Characters/Player/Male/Medium/Bodies/M_MED_RadioPaca/Textures/T_RadioPaca_Body_SoleNoise
  }
  n36 = StaticSwitchParameter(True=@n35.R, False=Constant()) {defaultValue=true, parameterName=Use Foot Mask?}
  NamedRerouteDeclaration(None=@n36) {name=Sole Cell Noise, nodeColor=<rgba 0 1 0.5781 1>}
  NamedRerouteDeclaration(None=Saturate(None=Step(Y=@n36) {constX=0.01})) {name=Sole Mask, nodeColor=<rgba 1 0.2578 0 1>}
  n37 = TextureSample() {samplerType=SAMPLERTYPE_LinearColor, texture=/BRCosmetics/Characters/Player/Male/Medium/Bodies/M_MED_RadioPaca/Textures/T_RadioPaca_Body_FX2}
  NamedRerouteDeclaration(None=StaticSwitchParameter(True=@n37.B, False=Constant()) {defaultValue=true, parameterName=Use Additional FX?}) {name=Foot Mask, nodeColor=<rgba 0 1 0.0625 1>}
  NamedRerouteDeclaration(None=@n16.Tech Art Mask) {name=TA Mask, nodeColor=<rgba 1 0.6328 0 1>}
  n38 = MaterialFunctionCall() {materialFunction=/Game/Packages/Fortress_SharedMaterials/Base_Material_Functions/BaseCharacterShaderFunctions/Audio/MF_AudioReactivity}
  n39 = ScalarParameter() {defaultValue=1, parameterName=Is Reactive?}
  n40 = MaterialFunctionCall!(
   InAmplitude=NamedRerouteDeclaration(None=Multiply(A=@n38.Amplitude, B=@n39)) {name=inAmplitude, nodeColor=<rgba 0 1 0.8594 1>}
   InSpectrum=NamedRerouteDeclaration(None=Multiply(A=@n38.Spectrum, B=@n39)) {name=inSpectrum, nodeColor=<rgba 1 0.5391 0 1>}
   CustomFakeSpectrum
   CustomFakeAmplitude
  ) {materialFunction=/BRCosmetics/Global/MaterialFunctions/AudioReactivity/MF_AudioReactivity_FrontendLoop}
  n41 = ScalarParameter() {defaultValue=1, parameterName=Is Reactive?}
  NamedRerouteDeclaration(None=LinearInterpolate(Alpha=Multiply(A=@n40.Amplitude, B=@n41)) {constA=1, constB=0}) {name=Audio On, nodeColor=<rgba 0.0547 1 0 1>}
  n42 = ScalarParameter() {parameterName=Seam Mask}
  NamedRerouteDeclaration(None=StaticSwitchParameter(True=Multiply(A=@n37.G, B=@n42), False=Constant()) {defaultValue=true, parameterName=Use Additional FX?}) {name=Seam Mask, nodeColor=<rgba 0.1719 0 1 1>}
  n43 = NamedRerouteUsage()
  n44 = ScalarParameter() {defaultValue=0.15, parameterName=Equalizer Transition}
  n45 = OneMinus(None=SmoothStep(Min=@n43, Max=Add(A=@n43, B=@n44), Value=NamedRerouteUsage()))
  NamedRerouteDeclaration(None=ComponentMask(None=@n45) {r=true}) {name=Spectrum, nodeColor=<rgba 0 1 0.9531 1>}
  NamedRerouteDeclaration(None=ComponentMask(None=@n16.Tech Art Mask) {g=true}) {name=Disc, nodeColor=<rgba 1 0.6328 0 1>}
  NamedRerouteDeclaration(None=StaticSwitchParameter(True=NamedRerouteUsage(), False=NamedRerouteUsage()) {parameterName=Slime Face Acc}) {name=Emissive Color - Out, nodeColor=<rgba 0.4766 1 0 1>}
  n46 = VectorParameter() {parameterName=Ray Tiling, defaultValue=<rgba 1 30.776 0 0>}
  n47 = Add(
   A=Multiply(A=NamedRerouteUsage(), B=ComponentMask(None=@n46.RGB) {g=true, r=true})
   B=ScalarParameter() {defaultValue=-0.0352, parameterName=Ray Offset}
  )
  NamedRerouteDeclaration(None=Frac(None=@n47)) {name=Ray Tiling, nodeColor=<rgba 0 1 0.8594 1>}
  n48 = NamedRerouteUsage()
  NamedRerouteDeclaration(None=Step(Y=ScalarParameter() {defaultValue=0.8, parameterName=Cell Noise - Step}, X=ComponentMask(None=@n48) {g=true}) {constX=0.01, constY=0.02}) {name=Stepped Cell Noise, nodeColor=<rgba 1 0 0.0469 1>}
  n49 = ScalarParameter() {defaultValue=0.8912, parameterName=Ray Edge Step}
  n50 = ScalarParameter() {defaultValue=-0.76, parameterName=Ray Edge Smoothness}
  n51 = NamedRerouteUsage()
  n52 = ComponentMask(None=VectorParameter() {parameterName=Ray Thickness, defaultValue=<rgba 2.0532 0.2574 0 0>}.RGB) {g=true, r=true}
  n53 = Multiply(
   A=SmoothStep(
    Min=@n49
    Max=Add(A=@n49, B=@n50)
    Value=MaterialFunctionCall(UVs=@n51, Size=@n52) {materialFunction=&MF_SDF_Box}.SDF
   )
   B=ScalarParameter() {defaultValue=0.7, parameterName="Ray  Opacity"}
  )
  NamedRerouteDeclaration(None=Multiply(A=NamedRerouteDeclaration(None=@n53) {name=Pre Glow Ray, nodeColor=<rgba 1 0.0469 0 1>}, B=NamedRerouteUsage())) {name=Rays, nodeColor=<rgba 0.1953 1 0 1>}
  n54 = Time()
  n55 = ScalarParameter() {defaultValue=0.005, parameterName=Scanline Speed}
  n56 = ComponentMask(None=Add(A=TextureCoordinate(), B=Multiply(A=@n54, B=@n55))) {g=true}
  n57 = ScalarParameter() {defaultValue=50, parameterName=Scanline Amount}
  n58 = Power(
   Base=MaterialFunctionCall!(Sine Phase=Multiply(A=@n56, B=@n57), Value 1, Value 2) {materialFunction=/Engine/Functions/Engine_MaterialFunctions02/Utility/Sine_Remapped}.Remapped Sine
   Exp=ScalarParameter() {defaultValue=1.0759, parameterName=Scanline Pow}
  )
  n59 = Fresnel(ExponentIn=ScalarParameter() {defaultValue=1, parameterName=Screen - Fresnel})
  n60 = Multiply(
   A=Multiply(A=Saturate(None=@n58), B=ScalarParameter() {parameterName=Scanline Intensity})
   B=Max(A=OneMinus(None=@n59), B=ScalarParameter() {parameterName=Screen - Fresnel Min})
  )
  n61 = NamedRerouteUsage()
  n62 = LinearInterpolate(
   A=ScalarParameter() {parameterName=Screen Intensity}
   B=ScalarParameter() {defaultValue=1, parameterName=Secondary - Screen Intensity}
   Alpha=@n61
  )
  n63 = Multiply(A=VectorParameter() {parameterName=Screen Color, defaultValue=<rgba 0.0029 0.0272 0.0938 1>}.RGB, B=@n62)
  n64 = Add(A=NamedRerouteUsage(), B=@n61)
  NamedRerouteDeclaration(None=Multiply(
    A=MaterialFunctionCall!(Blend=@n60, Base=@n63) {materialFunction=/Engine/Functions/Engine_MaterialFunctions03/Blends/Blend_ColorDodge}.Result
    B=Multiply(A=@n64, B=@n62)
   )) {name=Screen Effects, nodeColor=<rgba 0 0.3906 1 1>}
  NamedRerouteDeclaration(None=OneMinus(None=@n20)) {name=Ray Area, nodeColor=<rgba 1 0.6328 0 1>}
  n65 = ScalarParameter() {defaultValue=0.2126, parameterName=Ray Fade Start}
  n66 = ScalarParameter() {defaultValue=-0.05, parameterName=Ray Fade Smoothness}
  n67 = ComponentMask(None=NamedRerouteUsage()) {g=true}
  NamedRerouteDeclaration(None=SmoothStep(Min=@n65, Max=Add(A=@n65, B=@n66), Value=@n67)) {name=Ray Opacity Fade, nodeColor=<rgba 0.7813 0 1 1>}
  n68 = Ceil(None=AppendVector(A=NamedRerouteUsage(), B=Constant() {r=1}))
  n69 = Multiply(A=ComponentMask(None=VectorNoise(World Position=@n68)) {r=true}, B=Multiply(A=Time(), B=ScalarParameter() {parameterName=Ray Noise Speed}))
  n70 = Multiply(
   A=Max(A=Sine(None=@n69), B=ScalarParameter() {defaultValue=0.1, parameterName=Ray Min Noise})
   B=ScalarParameter() {defaultValue=1, parameterName=Ray Noise Intensity}
  )
  n71 = NamedRerouteUsage()
  n72 = LinearInterpolate(
   A=VectorParameter() {parameterName=Ray Colour, defaultValue=<rgba 0 0.1167 1 1>}.RGB
   B=VectorParameter() {parameterName=Ray Colour 2, defaultValue=<rgba 1.3399 13.9647 20 1>}.RGB
   Alpha=Saturate(None=Multiply(A=@n70, B=@n71))
  )
  n73 = ScalarParameter() {defaultValue=4, parameterName=Ray Intensity}
  n74 = Multiply(
   A=MaterialFunctionCall(InAmplitude=NamedRerouteUsage()) {materialFunction=&MF_AudioReactivity_FrontendLoop}.Amplitude
   B=ScalarParameter() {defaultValue=1, parameterName=Is Reactive?}
  )
  n75 = ScalarParameter() {defaultValue=8, parameterName=Ray Exponent}
  n76 = LinearInterpolate(B=Multiply(A=Power(Base=@n74, Exp=@n75), B=ScalarParameter() {defaultValue=8, parameterName=Ray Scale}), Alpha=Saturate(None=@n74)) {constA=1, constB=0}
  NamedRerouteDeclaration(None=Multiply(A=Multiply(A=@n72, B=@n73), B=@n76)) {name=Ray Color, nodeColor=<rgba 1 0 0.5156 1>}
  n77 = NamedRerouteUsage()
  n78 = LinearInterpolate(A=Multiply(A=ScalarParameter() {defaultValue=3, parameterName=Radial Audio Scale}, B=@n77), B=@n77, Alpha=NamedRerouteUsage()) {constA=0.6, constB=0.6}
  n79 = ScalarParameter() {defaultValue=0.18, parameterName=Radial Smoothness}
  n80 = NamedRerouteUsage()
  NamedRerouteDeclaration(None=SmoothStep(Min=@n78, Max=Add(A=@n78, B=@n79), Value=ComponentMask(None=@n80) {r=true})) {name=Radial Step, nodeColor=<rgba 1 0.6563 0 1>}
  n81 = ScalarParameter() {defaultValue=0.2118, parameterName=Ray Glow}
  n82 = ScalarParameter() {defaultValue=-0.2, parameterName=Ray Glow Smoothness}
  NamedRerouteDeclaration(None=SmoothStep(Min=@n81, Max=Add(A=@n81, B=@n82), Value=@n67)) {name=Ray Glow, nodeColor=<rgba 0.1719 0 1 1>}
  n83 = Time()
  n84 = Multiply(A=ScalarParameter() {defaultValue=4, parameterName=Stepped Speed}, B=ScalarParameter() {parameterName=Is Walking?})
  n85 = Add(A=NamedRerouteUsage(), B=Ceil(None=Multiply(A=@n83, B=@n84))) {constB=0}
  n86 = LinearInterpolate(
   A=StaticComponentMaskParameter(None=@n85) {defaultG=true, parameterName=Spectrum Axis}
   B=ComponentMask(None=@n85) {r=true}
   Alpha=NamedRerouteUsage()
  )
  n87 = NamedRerouteUsage()
  n88 = Divide(A=LinearInterpolate(A=StaticComponentMaskParameter(None=@n87) {defaultG=true, parameterName=Spectrum Axis}, B=@n87, Alpha=NamedRerouteUsage())) {constB=4}
  n89 = Constant() {r=4}
  n90 = Multiply(A=Frac(None=Multiply(A=@n86, B=@n88)), B=@n89)
  n91 = Power(Base=Multiply(A=Floor(None=@n90), B=Divide(B=@n89) {constB=4}))
  n92 = Step(Y=@n91) {constX=0.25}
  n93 = Step(Y=@n91) {constX=0.125}
  n94 = Step(Y=@n91) {constX=0}
  n95 = Max(A=@n93, B=@n94)
  n96 = Subtract(A=Step(Y=@n91), B=Max(A=@n92, B=@n95))
  n97 = NamedRerouteUsage()
  n98 = Multiply(A=Subtract(A=@n92, B=@n95), B=NamedRerouteUsage())
  n99 = Max(A=Multiply(A=Subtract(A=@n93, B=@n94), B=NamedRerouteUsage()), B=Multiply(A=NamedRerouteUsage(), B=@n94))
  n100 = NamedRerouteUsage()
  n101 = Power(
   Base=Max(A=Multiply(A=@n96, B=@n97), B=Max(A=@n98, B=@n99))
   Exp=LinearInterpolate(
    A=ScalarParameter() {defaultValue=1, parameterName=Spectrum Exponent}
    B=ScalarParameter() {defaultValue=1, parameterName=Secondary - Spectrum Exponent}
    Alpha=@n100
   )
  )
  n102 = LinearInterpolate(
   A=ScalarParameter() {defaultValue=0.25, parameterName=Spectrum Scale}
   B=ScalarParameter() {defaultValue=1.2, parameterName=Secondary - Spectrum Scale}
   Alpha=Max(A=@n100, B=NamedRerouteUsage())
  )
  NamedRerouteDeclaration(None=Saturate(None=Multiply(A=@n101, B=@n102))) {name=Audio Reactivity, nodeColor=<rgba 1 0.3281 0 1>}
  n103 = StaticSwitchParameter(True=NamedRerouteUsage(), False=NamedRerouteUsage()) {parameterName=Use Texture Cell Noise}
  n104 = Multiply(A=StaticSwitchParameter(True=Constant() {r=1}, False=NamedRerouteUsage()) {parameterName=Use Texture Cell Noise}, B=@n103)
  n105 = Multiply(A=@n104, B=@n45)
  NamedRerouteDeclaration(None=OneMinus(None=@n105)) {name=Current Boxes, nodeColor=<rgba 0.1484 1 0 1>}
  n106 = NamedRerouteUsage()
  NamedRerouteDeclaration(None=Step(Y=ComponentMask(None=@n106) {g=true}, X=ScalarParameter() {defaultValue=0.25, parameterName=Outer Step})) {name=Outer Step, nodeColor=<rgba 1 0 0.75 1>}
  n107 = Add(A=@n28, B=ComponentMask(None=VectorParameter() {parameterName=Front Mask Pos, defaultValue=<rgba -0.139 0.273 0 0>}.RGB) {g=true, r=true})
  n108 = Add(
   A=Step(Y=MaterialFunctionCall(UVs=@n107, Size=@n31.RGB) {materialFunction=&MF_SDF_Box}.SDF, X=@n32)
   B=@n33
  )
  NamedRerouteDeclaration(None=LinearInterpolate(A=LinearInterpolate(A=@n108, Alpha=@n34), Alpha=NamedRerouteUsage())) {name=Slime Variant, nodeColor=<rgba 0.7109 1 0 1>}
  n109 = Multiply(A=@n40.Spectrum, B=@n41)
  NamedRerouteDeclaration(None=ComponentMask(None=@n109) {a=true}) {name=High, nodeColor=<rgba 0.9688 0 1 1>}
  NamedRerouteDeclaration(None=ComponentMask(None=@n109) {g=true}) {name=Mid, nodeColor=<rgba 0 1 0.6484 1>}
  NamedRerouteDeclaration(None=ComponentMask(None=@n109) {b=true}) {name=Low, nodeColor=<rgba 1 0.3281 0 1>}
  NamedRerouteDeclaration(None=ComponentMask(None=@n109) {r=true}) {name=Bass, nodeColor=<rgba 0.0078 0 1 1>}
  NamedRerouteDeclaration(None=Multiply(A=NamedRerouteUsage(), B=ScalarParameter() {parameterName=Idle Sine})) {name=Audio Sine Animation, nodeColor=<rgba 1 0 0.1641 1>}
  n110 = NamedRerouteUsage()
  n111 = ScalarParameter() {defaultValue=0.3697, parameterName=Bass Level}
  n112 = NamedRerouteUsage()
  n113 = Multiply(A=Time() {period=0.8}, B=ScalarParameter() {defaultValue=2, parameterName=Global Idle Speed})
  n114 = Add(A=@n113) {constB=0.4}
  n115 = ScalarParameter() {defaultValue=0.05, parameterName=Param_4}
  n116 = ComponentMask(None=MaterialFunctionCall(
    Sine Phase=Multiply(A=@n114, B=@n115)
    Value 1=ScalarParameter() {defaultValue=-0.12, parameterName=Idle Sine - Min 01}
    Value 2=ScalarParameter() {defaultValue=0.1, parameterName=Idle Sine - Max 01}
   ) {materialFunction=&Sine_Remapped}.Remapped Sine) {r=true}
  n117 = NamedRerouteUsage()
  NamedRerouteDeclaration(None=Add(A=Multiply(A=@n110, B=@n111), B=LinearInterpolate(A=@n112, B=@n116, Alpha=@n117))) {name=Sine High, nodeColor=<rgba 0 0.4844 1 1>}
  n118 = ScalarParameter() {defaultValue=0.4534, parameterName=Low Level}
  n119 = NamedRerouteUsage()
  n120 = NamedRerouteUsage()
  n121 = Add(A=@n113) {constB=0.7}
  n122 = ScalarParameter() {defaultValue=0.2, parameterName=Param_5}
  n123 = ComponentMask(None=MaterialFunctionCall(
    Sine Phase=Multiply(A=@n121, B=@n122)
    Value 1=ScalarParameter() {defaultValue=-0.12, parameterName=Idle Sine - Min 02}
    Value 2=ScalarParameter() {defaultValue=0.12, parameterName=Idle Sine - Max 02}
   ) {materialFunction=&Sine_Remapped}.Remapped Sine) {r=true}
  n124 = NamedRerouteUsage()
  NamedRerouteDeclaration(None=Add(A=Multiply(A=@n118, B=@n119), B=LinearInterpolate(A=@n120, B=@n123, Alpha=@n124))) {name=Sine Low, nodeColor=<rgba 0.8047 1 0 1>}
  n125 = NamedRerouteUsage()
  NamedRerouteDeclaration(None=Step(Y=ComponentMask(None=@n125) {g=true}, X=ScalarParameter() {defaultValue=0.8342, parameterName=Ring Step})) {name=Ring Step, nodeColor=<rgba 0 0.7187 1 1>}
  n126 = ScalarParameter() {defaultValue=0.5515, parameterName=Mid Level}
  n127 = NamedRerouteUsage()
  n128 = NamedRerouteUsage()
  n129 = Add(A=@n113) {constA=0.5, constB=0.5}
  n130 = ScalarParameter() {defaultValue=0.15, parameterName=Param_6}
  n131 = ComponentMask(None=MaterialFunctionCall(
    Sine Phase=Multiply(A=@n129, B=@n130)
    Value 1=ScalarParameter() {defaultValue=-0.104, parameterName=Idle Sine - Min 03}
    Value 2=ScalarParameter() {defaultValue=0.08, parameterName=Idle Sine - Max 03}
   ) {materialFunction=&Sine_Remapped}.Remapped Sine) {r=true}
  n132 = NamedRerouteUsage()
  NamedRerouteDeclaration(None=Add(A=Multiply(A=@n126, B=@n127), B=LinearInterpolate(A=@n128, B=@n131, Alpha=@n132))) {name=Sine Mid, nodeColor=<rgba 0.875 0 1 1>}
  n133 = ScalarParameter() {defaultValue=0.5232, parameterName=High Level}
  n134 = NamedRerouteUsage()
  n135 = NamedRerouteUsage()
  n136 = Add(A=@n113) {constB=0.8}
  n137 = ScalarParameter() {defaultValue=0.1, parameterName=Param_7}
  n138 = ComponentMask(None=MaterialFunctionCall(
    Sine Phase=Multiply(A=@n136, B=@n137)
    Value 1=ScalarParameter() {defaultValue=-0.156, parameterName=Idle Sine - Min 04}
    Value 2=ScalarParameter() {defaultValue=0.1, parameterName=Idle Sine - Max 04}
   ) {materialFunction=&Sine_Remapped}.Remapped Sine) {r=true}
  n139 = NamedRerouteUsage()
  NamedRerouteDeclaration(None=Add(A=Multiply(A=@n133, B=@n134), B=LinearInterpolate(A=@n135, B=@n138, Alpha=@n139))) {name=Sine Bass, nodeColor=<rgba 0 1 0.5547 1>}
  NamedRerouteDeclaration(None=@n26.Radial Coordinates) {name=Ray UVs, nodeColor=<rgba 1 0 0.5391 1>}
  n140 = Constant()
  n141 = ScalarParameter() {defaultValue=-0.25, parameterName=Ray Rotation}
  NamedRerouteDeclaration(None=MaterialFunctionCall(Vector or UVs=MaterialFunctionCall(UVs=@n25, Rotation Center=@n140, "Rotation Angle (0-1)"=@n141) {materialFunction=&CustomRotator}.Rotated Values) {materialFunction=&VectorToRadialValue}.Radial Coordinates) {name=Rotated Ray UVs, nodeColor=<rgba 0 0.8594 1 1>}
  n142 = ScalarParameter() {parameterName=Param_10}
  n143 = ScalarParameter() {defaultValue=-0.01, parameterName=Param_11}
  NamedRerouteDeclaration(None=MaterialFunctionCall(Vector or UVs=BumpOffset(Coordinate=@n25, Height=@n142, HeightRatioInput=@n143) {heightRatio=0}) {materialFunction=&VectorToRadialValue}.Radial Coordinates) {name=Bump Offset UVs, nodeColor=<rgba 1 0.8203 0 1>}
  n144 = TextureSampleParameter2D(Bias=ScalarParameter() {defaultValue=-3, parameterName=FX Mip Bias}) {
   mipValueMode=TMVM_MipBias
   parameterName=Cell Noise Grayscale
   samplerType=SAMPLERTYPE_LinearGrayscale
   texture=/BRCosmetics/Gadgets/Assets/VinderTech_GliderChute/Glider_RadioPaca/Textures/T_RadioPaca_Glider_FX1
  }
  NamedRerouteDeclaration(None=@n144.R) {name=Cell Noise Texture, nodeColor=<rgba 0 0.9531 1 1>}
  n145 = Ceil(None=AppendVector(A=NamedRerouteUsage(), B=Constant() {r=1}))
  NamedRerouteDeclaration(None=ComponentMask(None=VectorNoise(World Position=@n145)) {r=true}) {name=Procedural Cell Noise, nodeColor=<rgba 1 0.7266 0 1>}
  n146 = ComponentMask(None=LinearInterpolate(
    A=VectorParameter() {parameterName=Spectrum Tiling, defaultValue=<rgba 70 33 0 0>}.RGB
    B=VectorParameter() {parameterName=Secondary - Spectrum Tiling, defaultValue=<rgba 50 100 0 0>}.RGB
    Alpha=NamedRerouteUsage()
   )) {g=true, r=true}
  NamedRerouteDeclaration(None=Multiply(A=NamedRerouteUsage(), B=NamedRerouteDeclaration(None=@n146) {name=Equalizer Tiling, nodeColor=<rgba 1 0 0.6328 1>})) {name=Cell Noise UVs, nodeColor=<rgba 0.4062 0 1 1>}
  NamedRerouteDeclaration(None=@n47) {name=Ray Cell Noise UV, nodeColor=<rgba 0 1 0.0859 1>}
  NamedRerouteDeclaration(None=Saturate(None=@n76)) {name=Ray Reactivity, nodeColor=<rgba 1 0 0.2344 1>}
  n147 = NamedRerouteUsage()
  n148 = Subtract(A=ScalarParameter() {defaultValue=0.25, parameterName=Random Noise - Min}, B=Subtract(A=NamedRerouteUsage(), B=OneMinus(None=@n147)))
  n149 = NamedRerouteUsage()
  n150 = NamedRerouteUsage()
  n151 = Saturate(None=Subtract(A=StaticSwitchParameter(True=@n149, False=@n150) {parameterName=Use Texture Cell Noise}, B=NamedRerouteUsage()))
  n152 = NamedRerouteUsage()
  n153 = Time()
  n154 = ScalarParameter() {defaultValue=1, parameterName=Cell Noise Speed}
  n155 = Sine(None=Multiply(A=Max(A=@n151, B=@n152), B=Multiply(A=@n153, B=@n154)))
  NamedRerouteDeclaration(None=Min(A=Max(A=@n148, B=@n155), B=ScalarParameter() {defaultValue=1, parameterName=Cell Noise Max})) {name=Cell Noise, nodeColor=<rgba 0.3125 1 0 1>}
  n156 = ScalarParameter() {defaultValue=0.2, parameterName=BG Position}
  n157 = ScalarParameter() {defaultValue=0.5, parameterName=BG Fade}
  n158 = Divide(A=VertexInterpolator(VS=LocalPosition() {localOrigin=InstancePreSkinning}.XYZ).PS, B=PreSkinnedLocalBounds().Max)
  n159 = OneMinus(None=NamedRerouteUsage())
  n160 = ScalarParameter() {parameterName=Use FX Gradient?}
  n161 = Constant() {r=1}
  n162 = OneMinus(None=NamedRerouteUsage())
  n163 = NamedRerouteUsage()
  n164 = LinearInterpolate(
   A=LinearInterpolate(A=ComponentMask(None=@n158) {b=true}, B=@n159, Alpha=LinearInterpolate(A=@n160, B=@n161, Alpha=@n162))
   B=@n159
   Alpha=@n163
  )
  n165 = ComponentMask(None=@n164) {r=true}
  NamedRerouteDeclaration(None=SmoothStep(Min=@n156, Max=Add(A=@n156, B=@n157), Value=OneMinus(None=@n165))) {name=BG, nodeColor=<rgba 0.5703 1 0 1>}
  n166 = LinearInterpolate(A=ScalarParameter() {defaultValue=-0.4, parameterName=Equalizer Start Offset}, Alpha=NamedRerouteUsage()) {constB=-1}
  n167 = ScalarParameter() {defaultValue=0.1048, parameterName=Screen - Equalizer Start Offset}
  NamedRerouteDeclaration(None=Add(A=@n164, B=LinearInterpolate(A=@n166, B=@n167, Alpha=@n163))) {name=Equalizer Gradient, nodeColor=<rgba 1 0.7734 0 1>}
  n168 = ScalarParameter() {defaultValue=0.42, parameterName=Equalizer Thickness}
  n169 = ScalarParameter() {defaultValue=0.1, parameterName=Equalizer Smoothness}
  n170 = Frac(None=Multiply(A=NamedRerouteUsage(), B=NamedRerouteUsage()))
  n171 = ComponentMask(None=LinearInterpolate(
    A=VectorParameter() {parameterName=Equalizer Box Size, defaultValue=<rgba 0.2514 0.4096 0 0>}.RGB
    B=VectorParameter() {parameterName=Secondary - Equalizer Box Size, defaultValue=<rgba 0.2404 0.1326 0 0>}.RGB
    Alpha=NamedRerouteUsage()
   )) {g=true, r=true}
  n172 = LinearInterpolate(
   A=ScalarParameter() {defaultValue=-0.3088, parameterName=Equalizer Box Roundness}
   B=ScalarParameter() {defaultValue=-0.2, parameterName=Secondary - Equalizer Box Roundness}
   Alpha=NamedRerouteUsage()
  )
  n173 = StaticBoolParameter() {parameterName="KeepAspectRatio (SB)"}
  n174 = Subtract(
   A=SmoothStep(
    Min=@n168
    Max=Add(A=@n168, B=@n169)
    Value=MaterialFunctionCall(UVs=@n170, Size=@n171, CornerRoundness=@n172, KeepAspectRatio=@n173) {materialFunction=&MF_SDF_Box}.SDF
   )
   B=NamedRerouteUsage()
  )
  NamedRerouteDeclaration(None=OneMinus(None=Saturate(None=@n174))) {name=Tiled Boxes, nodeColor=<rgba 0 0.8828 1 1>}
  n175 = GetMaterialAttributes(None=@n16.Result) {attributeGetTypes [
    #2
    #1
    #3
    #4
   ]}
  NamedRerouteDeclaration(None=@n175.Emissive Color) {name=Emissive Color - In, nodeColor=<rgba 0.5469 0 1 1>}
  n176 = VectorParameter() {parameterName=Arrow Color, defaultValue=<rgba 2.1704 5 0 1>}
  n177 = TextureSample!(UVs, Tex, Bias=ScalarParameter() {defaultValue=-5, parameterName=MipBias Arrows}, Apply View MipBias) {
   mipValueMode=TMVM_MipBias
   samplerType=SAMPLERTYPE_LinearColor
   texture=/BRCosmetics/Characters/Player/Male/Medium/Bodies/M_MED_RadioPaca/Skins/Slime/Textures/T_RadioPaca_Slime_FaceAcc_FX2
  }
  n178 = MaterialFunctionCall!(FlowID, FlowGradient=@n177.R, FlowMask=@n177.R) {materialFunction=/Game/Global/Materials/MaterialGym/Materials/Flow/MaterialFunctions/MF_Flow}
  n179 = Multiply(A=@n177.G, B=ScalarParameter() {defaultValue=0.25})
  n180 = Max(A=NamedRerouteUsage(), B=Max(A=Multiply(A=@n176.RGB, B=@n178.Result), B=Subtract(A=@n179, B=@n178.Result)))
  NamedRerouteDeclaration(None=Max(A=NamedRerouteUsage(), B=LinearInterpolate(B=@n180, Alpha=@n177.B))) {name=Slime Face Acc, nodeColor=<rgba 0 1 0.2266 1>}
  n181 = NamedRerouteUsage()
  n182 = ScalarParameter() {parameterName=Curve Index}
  n183 = ScalarParameter() {defaultValue=1, parameterName=Secondary - Curve Index}
  n184 = NamedRerouteUsage()
  n185 = TextureSample!(
   UVs=MaterialFunctionCall!(
    CurveIndex=LinearInterpolate(A=@n182, B=@n183, Alpha=@n184)
    CurveNum=ScalarParameter() {defaultValue=4, parameterName=Curve Amount}
    GreyMapIn=NamedRerouteUsage()
   ) {materialFunction=/Game/Effects/Niagara/MasterMaterials/Functions/Common/FXCommon_CurveAtlasUV}.UV
   Tex
   Apply View MipBias
  ) {samplerType=SAMPLERTYPE_LinearColor, texture=&/CA_RadioPaca}
  n186 = OneMinus(None=NamedRerouteUsage())
  n187 = NamedRerouteUsage()
  n188 = Multiply(
   A=@n185.RGB
   B=LinearInterpolate(
    A=ScalarParameter() {defaultValue=1, parameterName=Equalizer Brightness}
    B=ScalarParameter() {defaultValue=1, parameterName=Secondary - Brightness}
    Alpha=StaticSwitchParameter(True=@n186, False=@n187) {parameterName=Use Texture Cell Noise}
   )
  )
  n189 = NamedRerouteUsage()
  n190 = StaticSwitchParameter(True=Multiply(A=Multiply(A=@n188, B=@n189), B=NamedRerouteUsage()), False=@n188) {parameterName=Use Rays?}
  n191 = StaticSwitchParameter(True=LinearInterpolate(A=@n190, B=NamedRerouteUsage(), Alpha=NamedRerouteUsage()), False=@n190) {parameterName=Use Rays?}
  n192 = NamedRerouteUsage()
  NamedRerouteDeclaration(None=Add(A=LinearInterpolate(A=@n181, B=@n191, Alpha=@n192), B=NamedRerouteUsage())) {name=Combined Emissive, nodeColor=<rgba 1 0 0.0937 1>}
  NamedRerouteDeclaration(None=@n175.Normal) {name=Normal - In, nodeColor=<rgba 0.7344 1 0 1>}
  n193 = Saturate(None=NamedRerouteUsage())
  NamedRerouteDeclaration(None=LinearInterpolate(A=NamedRerouteUsage(), B=Constant3Vector() {constant=<rgba 0.5 0.5 1 0>}, Alpha=ComponentMask(None=@n193) {r=true})) {name=Normal - Out, nodeColor=<rgba 0 1 0.625 1>}
  NamedRerouteDeclaration(None=@n175.Base Color) {name=Base Color - In, nodeColor=<rgba 1 0.3047 0 1>}
  NamedRerouteDeclaration(None=@n175.MaterialAttributes) {name=Material Att - In, nodeColor=<rgba 0 0.0156 1 1>}
  NamedRerouteDeclaration(None=@n175.Specular) {name=Specular - In, nodeColor=<rgba 0.3359 1 0 1>}
  n194 = ComponentMask(None=NamedRerouteUsage()) {r=true}
  NamedRerouteDeclaration(None=LinearInterpolate(A=NamedRerouteUsage(), Alpha=@n194) {constB=0}) {name=Specular - Out, nodeColor=<rgba 1 0 0.6562 1>}
  n195 = NamedRerouteUsage()
  n196 = ScalarParameter() {defaultValue=1, parameterName=BC Darkening}
  n197 = StaticSwitchParameter(True=OneMinus(None=NamedRerouteUsage()), False=Constant()) {defaultValue=true, parameterName=Use Foot Mask?}
  n198 = Subtract(A=NamedRerouteUsage(), B=@n197)
  n199 = LinearInterpolate(
   A=LinearInterpolate(A=NamedRerouteUsage(), B=Multiply(A=@n195, B=@n196), Alpha=Saturate(None=@n198))
   Alpha=Saturate(None=Subtract(A=@n194, B=@n197))
  ) {constB=0}
  n200 = NamedRerouteUsage()
  NamedRerouteDeclaration(None=Saturate(None=Subtract(A=@n199, B=@n200))) {name=Base Color - Out, nodeColor=<rgba 0 0.9766 1 1>}
  n201 = Multiply(A=@n103, B=NamedRerouteUsage())
  n202 = ScalarParameter() {defaultValue=0.06, parameterName=BG Strength}
  NamedRerouteDeclaration(None=Multiply(A=Multiply(A=@n201, B=@n202), B=NamedRerouteUsage())) {name=Color, nodeColor=<rgba 1 0.7031 0 1>}
  n203 = Multiply(A=@n104, B=NamedRerouteUsage())
  n204 = Subtract(A=@n203, B=NamedRerouteUsage())
  NamedRerouteDeclaration(None=StaticSwitchParameter(True=Saturate(None=@n204), False=NamedRerouteDeclaration(None=@n203) {name=Equalizer Boxes, nodeColor=<rgba 1 0 0.7266 1>}) {parameterName=Use Rays?}) {name=Color Curve UVs, nodeColor=<rgba 0.3828 0 1 1>}
  n205 = NamedRerouteUsage()
  n206 = Multiply(A=StaticSwitchParameter(True=Multiply(A=@n205, B=@n105), False=@n105) {parameterName=Slime Variant?}, B=NamedRerouteUsage())
  n207 = NamedRerouteUsage()
  n208 = Saturate(None=StaticSwitchParameter(True=NamedRerouteUsage(), False=Subtract(A=@n206, B=@n207)) {parameterName=Use Texture Cell Noise})
  n209 = Fresnel() {baseReflectFraction=0, exponent=1}
  n210 = Power(Base=Saturate(None=OneMinus(None=@n209)), Exp=ScalarParameter() {defaultValue=4, parameterName=Equalizer Fresnel})
  n211 = ScalarParameter() {defaultValue=0.5, parameterName=Equalizer Fresnel Min}
  n212 = Saturate(None=Multiply(A=NamedRerouteDeclaration(None=@n208) {name=Pattern - Normal Mask, nodeColor=<rgba 0.9453 0 1 1>}, B=Max(A=@n210, B=@n211)))
  n213 = Max(A=Multiply(A=NamedRerouteUsage(), B=NamedRerouteUsage()), B=@n212)
  n214 = NamedRerouteUsage()
  n215 = Saturate(None=StaticSwitchParameter(True=Multiply(A=@n213, B=@n214), False=NamedRerouteDeclaration(None=@n212) {name=Equalizer Masks, nodeColor=<rgba 0 0.4141 1 1>}) {parameterName=Use Rays?})
  n216 = Subtract(A=@n215, B=NamedRerouteUsage())
  n217 = OneMinus(None=NamedRerouteUsage())
  NamedRerouteDeclaration(None=StaticSwitchParameter(True=Max(A=@n216, B=@n217), False=@n215) {defaultValue=true, parameterName=Use Foot Mask?}) {name=Combined Emissive Mask, nodeColor=<rgba 0 1 0.0625 1>}
  NamedRerouteDeclaration(None=ComponentMask(None=@n16.Tech Art Mask) {b=true, r=true}) {name=Flow 2D, nodeColor=<rgba 1 0 0.5156 1>}
  n218 = ComponentMask(None=NamedRerouteUsage()) {r=true}
  n219 = ScalarParameter() {parameterName=Screen Mask - FX1 or FX2?}
  NamedRerouteDeclaration(None=StaticSwitchParameter(True=LinearInterpolate(A=@n37.R, B=@n218, Alpha=@n219), False=Constant()) {defaultValue=true, parameterName=Use Additional FX?}) {name=Screen, nodeColor=<rgba 1 0.5625 0 1>}
 ]
}
