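# Physics asset for the RadioPaca male medium body's shirt, set up for secondary motion of the shirt's dynamic bones.
# Four chains (front and back, left and right) each join a tiny MSTR root capsule to two larger capsule segments.
# Constraints set tight twist and swing limits, with one swing axis loosening on the lower segment.
# Unconstrained pelvis, lower spine and thigh capsules provide body shapes for the cloth segments to collide against.
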PhysicsAsset /BRCosmetics/Characters/Player/Male/Medium/Bodies/M_MED_RadioPaca/Meshes/M_MED_RadioPaca_Shirt_Physics
bodies {
 dyn_shirt_back_l_01 [{
   shapeType=Capsule
   length=18.2435
   radius=5.84
   shapeName=dyn_shirt_back_l_01_capsule
   center=<4.0807 0 7>
   rotation=<rot -0.0003 0 105.4363>
  }]
 dyn_shirt_back_l_02 [{
   shapeType=Capsule
   length=16.9095
   radius=6.1673
   shapeName=dyn_shirt_back_l_02_capsule
   center=<3.0624 0 5>
   rotation=<rot -0.0003 0 105.4363>
  }]
 dyn_shirt_back_l_MSTR [{shapeType=Capsule, length=0.505, radius=0.505, shapeName=dyn_shirt_back_l_MSTR_capsule, rotation=<rot 0 -90 0>}]
 dyn_shirt_back_r_01 [{shapeType=Capsule, length=13.7945, radius=5.84, shapeName=dyn_shirt_back_r_01_capsule, center=<4.0807 0 7>, rotation=<rot -0.0001 0 78.7353>}]
 dyn_shirt_back_r_02 [{
   shapeType=Capsule
   length=12.6442
   radius=6.1673
   shapeName=dyn_shirt_back_r_02_capsule
   center=<3.0624 0 5>
   rotation=<rot -0.0001 0 78.7353>
  }]
 dyn_shirt_back_r_MSTR [{shapeType=Capsule, length=0.505, radius=0.505, shapeName=dyn_shirt_back_r_MSTR_capsule, rotation=<rot 0 -90 0>}]
 dyn_shirt_front_l_01 [{
   shapeType=Capsule
   length=14.1928
   radius=6.1136
   shapeName=dyn_shirt_front_01_capsule
   center=<4.9269 1.9 6.2>
   rotation=<rot -0.0002 179.9505 -68.6664>
  }]
 dyn_shirt_front_l_02 [{
   shapeType=Capsule
   length=18.3784
   radius=6.1437
   shapeName=dyn_shirt_front_02_capsule
   center=<3.9813 1.932 5.8>
   rotation=<rot -0.0002 -178.9856 -67.7651>
  }]
 dyn_shirt_front_l_MSTR [{shapeType=Capsule, length=0.5, radius=0.5, shapeName=dyn_shirt_front_l_MSTR_capsule}]
 dyn_shirt_front_r_01 [{
   shapeType=Capsule
   length=16.6263
   radius=6.1136
   shapeName=dyn_shirt_front_01_capsule
   center=<4.9269 -2 6.2>
   rotation=<rot -0.0001 179.9505 -102.2633>
  }]
 dyn_shirt_front_r_02 [{
   shapeType=Capsule
   length=18.3784
   radius=6.1437
   shapeName=dyn_shirt_front_02_capsule
   center=<4.0004 -2.068 5.8>
   rotation=<rot -0.0001 -178.9856 -101.3621>
  }]
 dyn_shirt_front_r_MSTR [{
   shapeType=Capsule
   length=0.505
   radius=0.505
   shapeName=dyn_shirt_front_r_MSTR_capsule
   center=<3.9978 -0.3175 0.9927>
   rotation=<rot 0 -0.3069 -76.2293>
  }]
 pelvis [{
   shapeType=Capsule
   length=10.9251
   radius=11.2093
   shapeName=pelvis_capsule
   center=<-10.4416 -0.0437 0.0185>
   rotation=<rot -0.7339 -90 0.3779>
  }]
 spine_01 [{shapeType=Capsule, length=15.8571, radius=9.609, shapeName=spine_01_capsule, center=<-11.646 -0.6146 0>, rotation=<rot 0 -90 0>}]
 thigh_l [{shapeType=Capsule, length=36.3211, radius=12.6201, shapeName=thigh_l_capsule, center=<-25.2429 -0.7167 0>, rotation=<rot 90 1.2762 0>}]
 thigh_r [{shapeType=Capsule, length=37.0497, radius=12.6746, shapeName=thigh_r_capsule, center=<25.4598 0.5858 0>, rotation=<rot 90 0.5971 0>}]
}
constraints [
 {bone1Name=dyn_shirt_front_l_01, bone2Name=dyn_shirt_front_l_MSTR, swing1LimitDegrees=10, swing2LimitDegrees=20, twistLimitDegrees=6}
 {bone1Name=dyn_shirt_front_l_02, bone2Name=dyn_shirt_front_l_01, swing1LimitDegrees=10, swing2LimitDegrees=40, twistLimitDegrees=6}
 {bone1Name=dyn_shirt_front_r_01, bone2Name=dyn_shirt_front_r_MSTR, swing1LimitDegrees=10, swing2LimitDegrees=20, twistLimitDegrees=6}
 {bone1Name=dyn_shirt_front_r_02, bone2Name=dyn_shirt_front_r_01, swing1LimitDegrees=10, swing2LimitDegrees=40, twistLimitDegrees=6}
 {bone1Name=dyn_shirt_back_l_01, bone2Name=dyn_shirt_back_l_MSTR, swing1LimitDegrees=10, swing2LimitDegrees=25, twistLimitDegrees=6}
 {bone1Name=dyn_shirt_back_l_02, bone2Name=dyn_shirt_back_l_01, swing1LimitDegrees=10, swing2LimitDegrees=40, twistLimitDegrees=6}
 {bone1Name=dyn_shirt_back_r_01, bone2Name=dyn_shirt_back_r_MSTR, swing1LimitDegrees=10, swing2LimitDegrees=25, twistLimitDegrees=6}
 {bone1Name=dyn_shirt_back_r_02, bone2Name=dyn_shirt_back_r_01, swing1LimitDegrees=10, swing2LimitDegrees=40, twistLimitDegrees=6}
]
