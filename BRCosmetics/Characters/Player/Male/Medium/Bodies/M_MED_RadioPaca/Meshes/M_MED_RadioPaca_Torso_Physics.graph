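# Physics asset for the RadioPaca medium male torso cosmetic, providing bodies for secondary motion on its hood, drawstrings and straps.
# Capsule chains (three links per drawstring and strap side, two for the hood) each start at a tiny MSTR root body.
# Angular swing and twist limits join the links, with the hood allowing the widest range.
# Four large, unconstrained spine capsules form the torso volume.
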PhysicsAsset /BRCosmetics/Characters/Player/Male/Medium/Bodies/M_MED_RadioPaca/Meshes/M_MED_RadioPaca_Torso_Physics
bodies {
 dyn_drawstring_l_01 [{shapeType=Capsule, length=5, radius=1, shapeName=dyn_drawstring_l_01_capsule, center=<1.5 0 0>, rotation=<rot 90 0 0>}]
 dyn_drawstring_l_02 [{shapeType=Capsule, length=5, radius=1, shapeName=dyn_drawstring_l_02_capsule, center=<1.5 0 0>, rotation=<rot 90 0 0>}]
 dyn_drawstring_l_03 [{shapeType=Capsule, length=5, radius=1, shapeName=dyn_drawstring_l_03_capsule, center=<1.5 0 0>, rotation=<rot 90 0 0>}]
 dyn_drawstring_l_MSTR [{shapeType=Capsule, length=0.505, radius=0.505, shapeName=dyn_drawstring_l_MSTR_capsule, rotation=<rot 0 -90 0>}]
 dyn_drawstring_r_01 [{shapeType=Capsule, length=4.853, radius=1.147, shapeName=dyn_drawstring_r_01_capsule, center=<1.5 0 0>, rotation=<rot 90 0 0>}]
 dyn_drawstring_r_02 [{shapeType=Capsule, length=4.853, radius=1.147, shapeName=dyn_drawstring_r_02_capsule, center=<1.5 0 0>, rotation=<rot 90 0 0>}]
 dyn_drawstring_r_03 [{shapeType=Capsule, length=4.853, radius=1.147, shapeName=dyn_drawstring_r_03_capsule, center=<1.5 0 0>, rotation=<rot 90 0 0>}]
 dyn_drawstring_r_MSTR [{shapeType=Capsule, length=0.505, radius=0.505, shapeName=dyn_drawstring_r_MSTR_capsule, rotation=<rot 0 -90 0>}]
 dyn_hood_01 [{shapeType=Capsule, length=3.534, radius=4.1319, shapeName=dyn_hood_01_capsule, center=<5.3181 0 1.3>, rotation=<rot 90 0 0>}]
 dyn_hood_02 [{shapeType=Capsule, length=3.4523, radius=3.7808, shapeName=dyn_hood_02_capsule, center=<2.9871 0 1.6>, rotation=<rot 90 0 0>}]
 dyn_hood_MSTR [{shapeType=Capsule, length=0.505, radius=0.505, shapeName=dyn_hood_MSTR_capsule, rotation=<rot 0 -90 0>}]
 dyn_strap_l_01 [{shapeType=Capsule, length=4, radius=2, shapeName=dyn_strap_l_01_capsule, center=<2 0 0>, rotation=<rot 90 0 0>}]
 dyn_strap_l_02 [{shapeType=Capsule, length=4, radius=2, shapeName=dyn_strap_l_02_capsule, center=<2 0 0>, rotation=<rot 90 0 0>}]
 dyn_strap_l_03 [{shapeType=Capsule, length=4, radius=2, shapeName=dyn_strap_l_03_capsule, center=<2 0 0>, rotation=<rot 90 0 0>}]
 dyn_strap_l_MSTR [{shapeType=Capsule, length=0.505, radius=0.505, shapeName=dyn_strap_l_MSTR_capsule, rotation=<rot 0 -90 0>}]
 dyn_strap_r_01 [{shapeType=Capsule, length=4, radius=2, shapeName=dyn_strap_r_01_capsule, center=<2 0 0>, rotation=<rot 90 0 0>}]
 dyn_strap_r_02 [{shapeType=Capsule, length=4, radius=2, shapeName=dyn_strap_r_02_capsule, center=<2 0 0>, rotation=<rot 90 0 0>}]
 dyn_strap_r_03 [{shapeType=Capsule, length=4, radius=2, shapeName=dyn_strap_r_03_capsule, center=<2 0 0>, rotation=<rot 90 0 0>}]
 dyn_strap_r_MSTR [{shapeType=Capsule, length=0.505, radius=0.505, shapeName=dyn_strap_r_MSTR_capsule, rotation=<rot 0 -90 0>}]
 spine_02 [{shapeType=Capsule, length=9.1535, radius=13.233, shapeName=spine_02_capsule, center=<2.4738 0 0>}]
 spine_03 [{shapeType=Capsule, length=6.1017, radius=14.2512, shapeName=spine_03_capsule, center=<5.1401 0 0>}]
 spine_04 [{shapeType=Capsule, length=5.3792, radius=13.8631, shapeName=spine_04_capsule, center=<2.8704 0 0>}]
 spine_05 [{shapeType=Capsule, length=8.3815, radius=13.4901, shapeName=spine_05_capsule}]
}
constraints [
 {bone1Name=dyn_drawstring_l_01, bone2Name=dyn_drawstring_l_MSTR, swing1LimitDegrees=20, swing2LimitDegrees=20, twistLimitDegrees=6}
 {bone1Name=dyn_drawstring_l_02, bone2Name=dyn_drawstring_l_01, swing1LimitDegrees=20, swing2LimitDegrees=20, twistLimitDegrees=6}
 {bone1Name=dyn_drawstring_l_03, bone2Name=dyn_drawstring_l_02, swing1LimitDegrees=20, swing2LimitDegrees=20, twistLimitDegrees=6}
 {bone1Name=dyn_drawstring_r_01, bone2Name=dyn_drawstring_r_MSTR, swing1LimitDegrees=20, swing2LimitDegrees=20, twistLimitDegrees=6}
 {bone1Name=dyn_drawstring_r_02, bone2Name=dyn_drawstring_r_01, swing1LimitDegrees=20, swing2LimitDegrees=20, twistLimitDegrees=6}
 {bone1Name=dyn_drawstring_r_03, bone2Name=dyn_drawstring_r_02, swing1LimitDegrees=20, swing2LimitDegrees=20, twistLimitDegrees=6}
 {bone1Name=dyn_strap_l_01, bone2Name=dyn_strap_l_MSTR, swing1LimitDegrees=20, swing2LimitDegrees=20, twistLimitDegrees=4}
 {bone1Name=dyn_strap_l_02, bone2Name=dyn_strap_l_01, swing1LimitDegrees=20, swing2LimitDegrees=20, twistLimitDegrees=4}
 {bone1Name=dyn_strap_l_03, bone2Name=dyn_strap_l_02, swing1LimitDegrees=20, swing2LimitDegrees=20, twistLimitDegrees=4}
 {bone1Name=dyn_strap_r_01, bone2Name=dyn_strap_r_MSTR, swing1LimitDegrees=20, swing2LimitDegrees=20, twistLimitDegrees=6}
 {bone1Name=dyn_strap_r_02, bone2Name=dyn_strap_r_01, swing1LimitDegrees=20, swing2LimitDegrees=20, twistLimitDegrees=6}
 {bone1Name=dyn_strap_r_03, bone2Name=dyn_strap_r_02, swing1LimitDegrees=20, swing2LimitDegrees=20, twistLimitDegrees=6}
 {bone1Name=dyn_hood_01, bone2Name=dyn_hood_MSTR, swing1LimitDegrees=30, swing2LimitDegrees=20, twistLimitDegrees=10}
 {bone1Name=dyn_hood_02, bone2Name=dyn_hood_01, swing1LimitDegrees=30, swing2LimitDegrees=20, twistLimitDegrees=10}
]
