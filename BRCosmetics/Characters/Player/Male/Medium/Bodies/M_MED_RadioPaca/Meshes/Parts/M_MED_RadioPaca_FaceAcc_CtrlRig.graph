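# Control Rig for the RadioPaca male medium face accessory that adds corrective and procedural motion to its hair tubes.
# A spherical pose reader on the head bone weights a transform offset on dyn_hair_MSTR and writes the head_back_SPR curve.
# The HairHarmonics function then waves eleven tube root bones (A–K) with alternating ItemHarmonics settings.
# HairHarmonicsSpeed scales wave speed and, clamped, sets wave maximum for A–D but minimum for E–K.
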ControlRigBlueprint /BRCosmetics/Characters/Player/Male/Medium/Bodies/M_MED_RadioPaca/Meshes/Parts/M_MED_RadioPaca_FaceAcc_CtrlRig
graphs [
 New Function_ContainedGraph {nodes [
   n1 = RigVMFunctionEntryNode()
   RigVMFunction_Sequence = &RigVMFunction_Sequence(ExecuteContext=@n1.ExecuteContext)
   n2 = RigVMVariableNode() {variable=HairHarmonicsSpeed}
   Clamp = &RigVMFunction_MathDoubleClamp(Maximum=2, Value=@n2.Value)
   n3 = /Script/RigVM.RigVMFunction_MathVectorMul(B=(X=1,Y=1,Z=1), A=RigVMVariableNode() {variable=ItemHarmonics_WaveSpeed}.Value, B.X=@n2.Value, B.Y=@n2.Value, B.Z=@n2.Value) {is=struct RigVMUnitNode, template="Multiply::Execute(in A,in B,out Result)"}
   /Script/ControlRig.RigUnit_ItemHarmonics(
    RotationOrder=YZX
    Targets=((Item=(Type=Bone,Name="dyn_hair_tube_A_01"),Ratio=20))
    WaveAmplitude=(X=0,Y=12,Z=14)
    WaveEase=Linear
    WaveFrequency=(X=0,Y=-2,Z=2)
    WaveMinimum=0
    WaveNoise=(X=0,Y=0.05,Z=0.1)
    WaveOffset=(X=0,Y=1,Z=3)
    ExecutePin=@RigVMFunction_Sequence.A
    WaveMaximum=@Clamp.Result
    WaveSpeed=@n3.Result
   ) {is=struct RigVMUnitNode}
   > &RigUnit_ItemHarmonics(
    RotationOrder=YZX
    Targets=((Item=(Type=Bone,Name="dyn_hair_tube_B_01"),Ratio=20))
    WaveAmplitude=(X=0,Y=15,Z=17)
    WaveEase=Linear
    WaveFrequency=(X=0,Y=2,Z=2)
    WaveMinimum=0
    WaveNoise=(X=0,Y=0.1,Z=0.05)
    WaveOffset=(X=0,Y=20,Z=1.5)
    WaveMaximum=@Clamp.Result
    WaveSpeed=@n3.Result
   )
   > &RigUnit_ItemHarmonics(
    RotationOrder=YZX
    Targets=((Item=(Type=Bone,Name="dyn_hair_tube_C_01"),Ratio=20))
    WaveAmplitude=(X=0,Y=12,Z=14)
    WaveEase=Linear
    WaveFrequency=(X=0,Y=-2,Z=2)
    WaveMinimum=0
    WaveNoise=(X=0,Y=0.05,Z=0.1)
    WaveOffset=(X=0,Y=1,Z=3)
    WaveMaximum=@Clamp.Result
    WaveSpeed=@n3.Result
   )
   > &RigUnit_ItemHarmonics(
    RotationOrder=YZX
    Targets=((Item=(Type=Bone,Name="dyn_hair_tube_D_01"),Ratio=20))
    WaveAmplitude=(X=0,Y=15,Z=17)
    WaveEase=Linear
    WaveFrequency=(X=0,Y=2,Z=2)
    WaveMinimum=0
    WaveNoise=(X=0,Y=0.1,Z=0.05)
    WaveOffset=(X=0,Y=20,Z=1.5)
    WaveMaximum=@Clamp.Result
    WaveSpeed=@n3.Result
   )
   > &RigUnit_ItemHarmonics(
    RotationOrder=YZX
    Targets=((Item=(Type=Bone,Name="dyn_hair_tube_E_01"),Ratio=20))
    WaveAmplitude=(X=0,Y=12,Z=14)
    WaveEase=Linear
    WaveFrequency=(X=0,Y=-2,Z=2)
    WaveMaximum=1
    WaveNoise=(X=0,Y=0.05,Z=0.1)
    WaveOffset=(X=0,Y=1,Z=3)
    WaveMinimum=@Clamp.Result
    WaveSpeed=@n3.Result
   )
   > &RigUnit_ItemHarmonics(
    RotationOrder=YZX
    Targets=((Item=(Type=Bone,Name="dyn_hair_tube_F_01"),Ratio=20))
    WaveAmplitude=(X=0,Y=15,Z=17)
    WaveEase=Linear
    WaveFrequency=(X=0,Y=2,Z=2)
    WaveMaximum=1
    WaveNoise=(X=0,Y=0.1,Z=0.05)
    WaveOffset=(X=0,Y=20,Z=1.5)
    WaveMinimum=@Clamp.Result
    WaveSpeed=@n3.Result
   )
   > &RigUnit_ItemHarmonics(
    RotationOrder=YZX
    Targets=((Item=(Type=Bone,Name="dyn_hair_tube_G_01"),Ratio=20))
    WaveAmplitude=(X=0,Y=12,Z=14)
    WaveEase=Linear
    WaveFrequency=(X=0,Y=-2,Z=2)
    WaveMaximum=1
    WaveNoise=(X=0,Y=0.05,Z=0.1)
    WaveOffset=(X=0,Y=1,Z=3)
    WaveMinimum=@Clamp.Result
    WaveSpeed=@n3.Result
   )
   > &RigUnit_ItemHarmonics(
    RotationOrder=YZX
    Targets=((Item=(Type=Bone,Name="dyn_hair_tube_H_01"),Ratio=20))
    WaveAmplitude=(X=0,Y=15,Z=17)
    WaveEase=Linear
    WaveFrequency=(X=0,Y=2,Z=2)
    WaveMaximum=1
    WaveNoise=(X=0,Y=0.1,Z=0.05)
    WaveOffset=(X=0,Y=20,Z=1.5)
    WaveMinimum=@Clamp.Result
    WaveSpeed=@n3.Result
   )
   > &RigUnit_ItemHarmonics(
    RotationOrder=YZX
    Targets=((Item=(Type=Bone,Name="dyn_hair_tube_I_01"),Ratio=20))
    WaveAmplitude=(X=0,Y=12,Z=14)
    WaveEase=Linear
    WaveFrequency=(X=0,Y=2,Z=2)
    WaveMaximum=1
    WaveNoise=(X=0,Y=0.05,Z=0.1)
    WaveOffset=(X=0,Y=20,Z=1.5)
    WaveMinimum=@Clamp.Result
    WaveSpeed=@n3.Result
   )
   > &RigUnit_ItemHarmonics(
    RotationOrder=YZX
    Targets=((Item=(Type=Bone,Name="dyn_hair_tube_J_01"),Ratio=20))
    WaveAmplitude=(X=0,Y=15,Z=17)
    WaveEase=Linear
    WaveFrequency=(X=0,Y=-2,Z=2)
    WaveMaximum=1
    WaveNoise=(X=0,Y=0.1,Z=0.05)
    WaveOffset=(X=0,Y=1,Z=3)
    WaveMinimum=@Clamp.Result
    WaveSpeed=@n3.Result
   )
   > &RigUnit_ItemHarmonics(
    RotationOrder=YZX
    Targets=((Item=(Type=Bone,Name="dyn_hair_tube_K_01"),Ratio=20))
    WaveAmplitude=(X=0,Y=12,Z=14)
    WaveEase=Linear
    WaveFrequency=(X=0,Y=2,Z=2)
    WaveMaximum=1
    WaveNoise=(X=0,Y=0.05,Z=0.1)
    WaveOffset=(X=0,Y=20,Z=1.5)
    WaveMinimum=@Clamp.Result
    WaveSpeed=@n3.Result
   )
   RigVMFunctionReturnNode(ExecuteContext=@RigVMFunction_Sequence.B)
  ]}
 RigVMFunctionLibrary {nodes [HairHarmonics = RigVMCollapseNode(ExecuteContext=())]}
 RigVMModel {nodes [
   RigUnit_BeginExecution = &RigUnit_BeginExecution()
   n1 = &RigUnit_SphericalPoseReader(
    ActiveRegionScaleFactors=(PositiveWidth=1,NegativeWidth=1,PositiveHeight=1,NegativeHeight=1)
    ActiveRegionSize=0.1
    Debug=(bDrawDebug=false,bDraw2D=False,bDrawLocalAxes=False,DebugScale=25,DebugSegments=20,DebugThickness=0.25)
    DriverAxis=(X=0,Y=1,Z=0)
    DriverItem=(Type=Bone,Name="head")
    FalloffRegionScaleFactors=(PositiveWidth=1,NegativeWidth=1,PositiveHeight=1,NegativeHeight=1)
    FalloffSize=0.3
    FlipHeightScaling=False
    FlipWidthScaling=False
    OptionalParentItem=(Type=Bone,Name="None")
    RotationOffset=(X=90,Y=0,Z=-60)
    ExecutePin=@RigUnit_BeginExecution.ExecutePin
   )
   > &RigUnit_ModifyTransforms(
    ItemToModify=((Item=(Type=Bone,Name="dyn_hair_MSTR"),Transform=(Rotation=(X=0,Y=0.3079,Z=0,W=-0.9514),Translation=(X=1.4536,Y=0,Z=-0.9472),Scale3D=(X=1,Y=1,Z=1))))
    WeightMinimum=0
    Weight=@n1.OutputParam
   )
   > /Script/ControlRig.RigUnit_SetCurveValue(Curve=head_back_SPR, Value=@n1.OutputParam) {is=struct RigVMUnitNode}
   > HairHarmonics_0 = RigVMFunctionReferenceNode()
   RigVMFunction_AlphaInterp = &RigVMFunction_AlphaInterp(InterpSpeedDecreasing=10, InterpSpeedIncreasing=10, Scale=1, bClampResult=False, bInterpResult=False, Value=@n1.OutputParam)
  ]}
]
hierarchy [
 root {loc=<0 0 0>, rot=<0 0 0 1>}
 pelvis {parent=root, loc=<0 -0.7071 93.1191>, rot=<0.0127 0.707 -0.0127 -0.707>}
 spine_01 {parent=pelvis, loc=<6.7494 -0.6388 0>, rot=<0 0 0.034 -0.9994>}
 spine_02 {parent=spine_01, loc=<7.2518 -0.035 0>, rot=<0 0 0.0238 -0.9997>}
 spine_03 {parent=spine_02, loc=<6.0104 0 0>, rot=<0 0 0.0149 -0.9999>}
 spine_04 {parent=spine_03, loc=<8.1024 0 0>, rot=<0 0 -0.0132 -0.9999>}
 spine_05 {parent=spine_04, loc=<10.0146 0 0>, rot=<0 0 -0.0015 -1>}
 neck_01 {parent=spine_05, loc=<11.3138 0.1529 0>, rot=<0 0 -0.2003 -0.9797>}
 neck_02 {parent=neck_01, loc=<4.5627 0 0>, rot=<0 0 0.0504 -0.9987>}
 head {parent=neck_02, loc=<5.3914 0 0>, rot=<0 0 0.115 -0.9934>}
 dyn_nose_ring_MSTR {parent=head, loc=<4.6528 21.6513 0>, rot=<0.4068 0.5783 0.5783 0.4068>}
 dyn_nose_ring {parent=dyn_nose_ring_MSTR, loc=<0 0 0>, rot=<0 0 0 1>}
 dyn_hair_MSTR {parent=head, loc=<10.6232 -8.1699 0>, rot=<0.5649 -0.4252 0.4252 -0.5649>}
 dyn_hair_knot {parent=dyn_hair_MSTR, loc=<0 0 0>, rot=<0 0 0 1>}
 dyn_hair_tube_A_MSTR {parent=dyn_hair_knot, loc=<-0.3795 3.4422 1.8954>, rot=<-0.1597 0.2416 -0.2728 -0.9175>}
 dyn_hair_tube_A_01 {parent=dyn_hair_tube_A_MSTR, loc=<0 0 0>, rot=<0 0 0 1>}
 dyn_hair_tube_A_02 {parent=dyn_hair_tube_A_01, loc=<5.2044 0 0>, rot=<0.0458 -0.3481 -0.0671 -0.9339>}
 dyn_hair_tube_A_03 {parent=dyn_hair_tube_A_02, loc=<4.8782 0 0>, rot=<-0.0084 -0.4021 0.0468 -0.9143>}
 dyn_hair_tube_A_04 {parent=dyn_hair_tube_A_03, loc=<4.4843 0 0>, rot=<-0.0173 -0.1907 0.0303 -0.981>}
 dyn_hair_tube_B_MSTR {parent=dyn_hair_knot, loc=<0.1694 3.4234 -1.2071>, rot=<0.0116 -0.287 -0.3214 -0.9023>}
 dyn_hair_tube_B_01 {parent=dyn_hair_tube_B_MSTR, loc=<0 0 0>, rot=<0 0 0 1>}
 dyn_hair_tube_B_02 {parent=dyn_hair_tube_B_01, loc=<5.2587 0 0>, rot=<0.0647 -0.2795 0.0205 -0.9577>}
 dyn_hair_tube_B_03 {parent=dyn_hair_tube_B_02, loc=<5.5717 0 0>, rot=<0.0342 -0.2054 -0.0021 -0.9781>}
 dyn_hair_tube_B_04 {parent=dyn_hair_tube_B_03, loc=<5.0052 0 0>, rot=<0.0225 -0.1078 -0.0311 -0.9934>}
 dyn_hair_tube_C_MSTR {parent=dyn_hair_knot, loc=<0.6224 2.0098 3.358>, rot=<0.0673 0.4311 -0.0022 -0.8998>}
 dyn_hair_tube_C_01 {parent=dyn_hair_tube_C_MSTR, loc=<0 0 0>, rot=<0 0 0 1>}
 dyn_hair_tube_C_02 {parent=dyn_hair_tube_C_01, loc=<3.8333 0 0>, rot=<0.0346 -0.1444 0.0493 -0.9877>}
 dyn_hair_tube_C_03 {parent=dyn_hair_tube_C_02, loc=<3.6592 0 0>, rot=<-0.0127 -0.1185 0.0265 -0.9925>}
 dyn_hair_tube_D_MSTR {parent=dyn_hair_knot, loc=<0.4356 2.6012 -2.8246>, rot=<0.0431 -0.4957 -0.1651 -0.8516>}
 dyn_hair_tube_D_01 {parent=dyn_hair_tube_D_MSTR, loc=<0 0 0>, rot=<0 0 0 1>}
 dyn_hair_tube_D_02 {parent=dyn_hair_tube_D_01, loc=<3.7871 0 0>, rot=<-0.0455 -0.2761 0.106 -0.9542>}
 dyn_hair_tube_D_03 {parent=dyn_hair_tube_D_02, loc=<3.1704 0 0>, rot=<-0.0546 -0.0583 0.0405 -0.996>}
 dyn_hair_tube_E_MSTR {parent=dyn_hair_knot, loc=<1.1013 0.3389 3.8605>, rot=<0.1675 0.2663 0.2691 -0.9103>}
 dyn_hair_tube_E_01 {parent=dyn_hair_tube_E_MSTR, loc=<0 0 0>, rot=<0 0 0 1>}
 dyn_hair_tube_E_02 {parent=dyn_hair_tube_E_01, loc=<5.1166 0 0>, rot=<0.0072 -0.1849 0.0094 -0.9827>}
 dyn_hair_tube_E_03 {parent=dyn_hair_tube_E_02, loc=<4.4901 0 0>, rot=<-0.0049 -0.1312 -0.0141 -0.9912>}
 dyn_hair_tube_F_MSTR {parent=dyn_hair_knot, loc=<1.7023 -1.1387 -0.2852>, rot=<0.1772 0.0211 0.1545 -0.9717>}
 dyn_hair_tube_F_01 {parent=dyn_hair_tube_F_MSTR, loc=<0 0 0>, rot=<0 0 0 1>}
 dyn_hair_tube_F_02 {parent=dyn_hair_tube_F_01, loc=<4.2741 0 0>, rot=<-0.0217 -0.0508 0.0567 -0.9969>}
 dyn_hair_tube_F_03 {parent=dyn_hair_tube_F_02, loc=<3.697 0 0>, rot=<-0.028 0.0024 0.0531 -0.9982>}
 dyn_hair_tube_G_MSTR {parent=dyn_hair_knot, loc=<2.567 -1.539 -3.5161>, rot=<-0.0073 -0.4043 0.0504 -0.9132>}
 dyn_hair_tube_G_01 {parent=dyn_hair_tube_G_MSTR, loc=<0 0 0>, rot=<0 0 0 1>}
 dyn_hair_tube_G_02 {parent=dyn_hair_tube_G_01, loc=<5.285 0 0>, rot=<-0.059 -0.2551 -0.0611 -0.9632>}
 dyn_hair_tube_G_03 {parent=dyn_hair_tube_G_02, loc=<4.1652 0 0>, rot=<-0.0055 -0.1769 -0.0633 -0.9822>}
 dyn_hair_tube_G_04 {parent=dyn_hair_tube_G_03, loc=<3.7382 0 0>, rot=<-0.0192 -0.0809 -0.0286 -0.9961>}
 dyn_hair_tube_H_MSTR {parent=dyn_hair_knot, loc=<5.2919 -1.8466 0.079>, rot=<0.1041 -0.0737 0.5662 -0.8143>}
 dyn_hair_tube_H_01 {parent=dyn_hair_tube_H_MSTR, loc=<0 0 0>, rot=<0 0 0 1>}
 dyn_hair_tube_H_02 {parent=dyn_hair_tube_H_01, loc=<5.317 0 0>, rot=<0.0202 -0.3246 -0.0825 -0.942>}
 dyn_hair_tube_H_03 {parent=dyn_hair_tube_H_02, loc=<4.1795 0 0>, rot=<0.0237 -0.1686 -0.0657 -0.9832>}
 dyn_hair_tube_I_MSTR {parent=dyn_hair_knot, loc=<2.1356 -4.4051 -5.0701>, rot=<-0.0565 -0.6123 0.1355 -0.7769>}
 dyn_hair_tube_I_01 {parent=dyn_hair_tube_I_MSTR, loc=<0 0 0>, rot=<0 0 0 1>}
 dyn_hair_tube_I_02 {parent=dyn_hair_tube_I_01, loc=<3.4423 0 0>, rot=<-0.0954 -0.146 -0.039 -0.9839>}
 dyn_hair_tube_I_03 {parent=dyn_hair_tube_I_02, loc=<3.5302 0 0>, rot=<0 0 0 1>}
 dyn_hair_tube_J_MSTR {parent=dyn_hair_knot, loc=<3.433 -5.1302 0.6625>, rot=<-0.2989 -0.243 0.7796 -0.4937>}
 dyn_hair_tube_J_01 {parent=dyn_hair_tube_J_MSTR, loc=<0 0 0>, rot=<0 0 0 1>}
 dyn_hair_tube_J_02 {parent=dyn_hair_tube_J_01, loc=<4.1405 0 0>, rot=<-0.0028 -0.1148 -0.2258 -0.9674>}
 dyn_hair_tube_J_03 {parent=dyn_hair_tube_J_02, loc=<4.3501 0 0>, rot=<0.0758 0.0068 -0.1458 -0.9864>}
 dyn_hair_tube_K_MSTR {parent=dyn_hair_knot, loc=<2.0954 -5.3851 -3.6486>, rot=<-0.1424 -0.4268 0.3413 -0.8253>}
 dyn_hair_tube_K_01 {parent=dyn_hair_tube_K_MSTR, loc=<0 0 0>, rot=<0 0 0 1>}
 dyn_hair_tube_K_02 {parent=dyn_hair_tube_K_01, loc=<5.6136 0 0>, rot=<-0.0358 -0.2536 -0.0048 -0.9666>}
 dyn_hair_tube_K_03 {parent=dyn_hair_tube_K_02, loc=<4.3595 0 0>, rot=<0.0155 -0.1869 -0.0129 -0.9822>}
 dyn_hair_tube_K_04 {parent=dyn_hair_tube_K_03, loc=<4.4783 0 0>, rot=<-0.0382 -0.0608 -0.0178 -0.9973>}
 clavicle_l {parent=spine_05, loc=<4.3575 -0.1466 -2.3223>, rot=<0.0465 -0.7289 -0.009 -0.6829>}
 upperarm_l {parent=clavicle_l, loc=<15.5111 -0.0001 0>, rot=<0.0226 -0.3234 -0.0088 -0.9459>}
 upperarm_twist_01_l {parent=upperarm_l, loc=<0.0041 0 0>, rot=<0 0 0 -1>}
 deform_upperarm_up_base_l {parent=upperarm_twist_01_l, loc=<-3.2216 -1.1391 3.3311>, rot=<0 0 0 -1>}
 deform_upperarm_up_l {parent=deform_upperarm_up_base_l, loc=<0 0 0>, rot=<0 0 0 1>}
 clavicle_r {parent=spine_05, loc=<4.3575 -0.1466 2.3223>, rot=<0.7289 0.0465 -0.6829 0.009>}
 upperarm_r {parent=clavicle_r, loc=<-15.5111 0.0001 0>, rot=<0.0226 -0.3234 -0.0088 -0.9459>}
 upperarm_twist_01_r {parent=upperarm_r, loc=<-0.0041 0 0>, rot=<0 0 0 -1>}
 deform_upperarm_up_base_r {parent=upperarm_twist_01_r, loc=<3.2214 1.1391 -3.3308>, rot=<0 0 0 -1>}
 deform_upperarm_up_r {parent=deform_upperarm_up_base_r, loc=<0 0 0>, rot=<0 0 0 -1>}
 deform_pec_aim_base_l {parent=spine_04, loc=<7.8732 12.6388 -2.4123>, rot=<0.0102 -0.4602 0.3979 -0.7936>}
 deform_pec_aim_l {parent=deform_pec_aim_base_l, loc=<0 0 0>, rot=<0 0 0 1>}
 deform_pec_base_l {parent=deform_pec_aim_l, loc=<10.766 0 0>, rot=<0 0 0 1>}
 deform_pec_l {parent=deform_pec_base_l, loc=<0.055 0 0>, rot=<0 0 0 1>}
 deform_pec_aim_base_r {parent=spine_04, loc=<7.8727 12.6388 2.4123>, rot=<-0.4602 -0.0102 0.7936 0.3979>}
 deform_pec_aim_r {parent=deform_pec_aim_base_r, loc=<0 0 0>, rot=<0 0 0 1>}
 deform_pec_base_r {parent=deform_pec_aim_r, loc=<-10.7662 -0.0006 -0.0007>, rot=<0 0 0 1>}
 deform_pec_r {parent=deform_pec_base_r, loc=<0 0 0>, rot=<0 0 0 1>}
 head_back_SPR {type=Curve, loc=<0 0 0>, rot=<0 0 0 1>}
]
variables [
 HairHarmonicsSpeed {type=double, default=0}
 isSkydive {type=bool, default=False}
]
