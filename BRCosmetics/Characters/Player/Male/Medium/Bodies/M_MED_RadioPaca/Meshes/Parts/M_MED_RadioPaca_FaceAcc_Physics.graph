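# Physics asset for the RadioPaca medium male character's face accessory, defining secondary motion for eleven hair-tube bone chains (A to K).
# Each chain links three or four capsule bodies in series from a small MSTR root body.
# Constraints keep twist tight on every link and widen swing limits from the third link onward.
# Large unconstrained capsules on the clavicles and upper spine bones provide body volumes around the shoulders and chest.
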PhysicsAsset /BRCosmetics/Characters/Player/Male/Medium/Bodies/M_MED_RadioPaca/Meshes/Parts/M_MED_RadioPaca_FaceAcc_Physics
bodies {
 clavicle_l [{
   shapeType=Capsule
   length=12.3832
   radius=12.6168
   shapeName=clavicle_l_capsule
   center=<5.1646 0 -2.0806>
   rotation=<rot 78.2454 -0.0001 -0.0001>
  }]
 clavicle_r [{
   shapeType=Capsule
   length=12.3832
   radius=12.6168
   shapeName=clavicle_r_capsule
   center=<-5.4397 0 1.9725>
   rotation=<rot 78.8726 -0.0003 -0.0003>
  }]
 dyn_hair_MSTR [{shapeType=Capsule, length=0.505, radius=0.505, shapeName=dyn_hair_MSTR_capsule, rotation=<rot 0 -90 0>}]
 dyn_hair_tube_A_01 [{
   shapeType=Capsule
   length=3.9104
   radius=1.25
   shapeName=dyn_hair_tube_A_01_capsule
   center=<1.9753 0 0.0884>
   rotation=<rot 83.0504 -179.9997 -179.9997>
  }]
 dyn_hair_tube_A_02 [{
   shapeType=Capsule
   length=5
   radius=1.25
   shapeName=dyn_hair_tube_A_02_capsule
   center=<1.25 0 0>
   rotation=<rot 72.7243 -179.9999 -179.9999>
  }]
 dyn_hair_tube_A_03 [{
   shapeType=Capsule
   length=5
   radius=1.25
   shapeName=dyn_hair_tube_A_03_capsule
   center=<1.25 0 0>
   rotation=<rot 81.4231 -179.9998 -179.9998>
  }]
 dyn_hair_tube_A_04 [{shapeType=Capsule, length=5, radius=1.25, shapeName=dyn_hair_tube_A_04_capsule, center=<1.25 0 0>, rotation=<rot 90 0 0>}]
 dyn_hair_tube_A_MSTR [{shapeType=Capsule, length=0.505, radius=0.505, shapeName=dyn_hair_tube_A_MSTR_capsule, rotation=<rot 0 -90 0>}]
 dyn_hair_tube_B_01 [{
   shapeType=Capsule
   length=4.8221
   radius=1.4279
   shapeName=dyn_hair_tube_B_01_capsule
   center=<2.304 0 0.0642>
   rotation=<rot 86.5153 -179.9995 -179.9996>
  }]
 dyn_hair_tube_B_02 [{
   shapeType=Capsule
   length=4.8221
   radius=1.4279
   shapeName=dyn_hair_tube_B_02_capsule
   center=<1.25 0 0>
   rotation=<rot 81.1147 -179.9998 -179.9999>
  }]
 dyn_hair_tube_B_03 [{
   shapeType=Capsule
   length=4.8221
   radius=1.4279
   shapeName=dyn_hair_tube_B_03_capsule
   center=<1.25 0 0>
   rotation=<rot 84.9157 -179.9999 -180>
  }]
 dyn_hair_tube_B_04 [{shapeType=Capsule, length=4.8221, radius=1.4279, shapeName=dyn_hair_tube_B_04_capsule, center=<1.8992 0 0>, rotation=<rot 90 0 0>}]
 dyn_hair_tube_B_MSTR [{shapeType=Capsule, length=0.505, radius=0.505, shapeName=dyn_hair_tube_B_MSTR_capsule, rotation=<rot 0 -90 0>}]
 dyn_hair_tube_C_01 [{shapeType=Capsule, length=3, radius=1.25, shapeName=dyn_hair_tube_C_01_capsule, center=<1.25 0 0>, rotation=<rot 90 0 0>}]
 dyn_hair_tube_C_02 [{
   shapeType=Capsule
   length=3.2597
   radius=1.25
   shapeName=dyn_hair_tube_C_02_capsule
   center=<1.25 0 0>
   rotation=<rot 88.0096 179.9999 179.9999>
  }]
 dyn_hair_tube_C_03 [{shapeType=Capsule, length=3, radius=1.25, shapeName=dyn_hair_tube_C_03_capsule, center=<0.9915 0 0>, rotation=<rot 90 0 0>}]
 dyn_hair_tube_C_MSTR [{shapeType=Capsule, length=0.505, radius=0.505, shapeName=dyn_hair_tube_C_MSTR_capsule, rotation=<rot 0 -90 0>}]
 dyn_hair_tube_D_01 [{shapeType=Capsule, length=3, radius=1.25, shapeName=dyn_hair_tube_D_01_capsule, center=<1.25 0 0>, rotation=<rot 90 0 0>}]
 dyn_hair_tube_D_02 [{shapeType=Capsule, length=3, radius=1.25, shapeName=dyn_hair_tube_D_02_capsule, center=<1.25 0 0>, rotation=<rot 87.5199 179.9998 179.9998>}]
 dyn_hair_tube_D_03 [{shapeType=Capsule, length=3, radius=1.25, shapeName=dyn_hair_tube_D_03_capsule, center=<1.25 0 0>, rotation=<rot 90 0 0>}]
 dyn_hair_tube_D_MSTR [{shapeType=Capsule, length=0.505, radius=0.505, shapeName=dyn_hair_tube_D_MSTR_capsule, rotation=<rot 0 -90 0>}]
 dyn_hair_tube_E_01 [{
   shapeType=Capsule
   length=5
   radius=1.25
   shapeName=dyn_hair_tube_E_01_capsule
   center=<0.6024 0 -0.0219>
   rotation=<rot 85.4377 179.9992 179.9991>
  }]
 dyn_hair_tube_E_02 [{
   shapeType=Capsule
   length=4.2374
   radius=1.25
   shapeName=dyn_hair_tube_E_02_capsule
   center=<0.5905 0 0.0498>
   rotation=<rot 80.5467 179.9994 179.9993>
  }]
 dyn_hair_tube_E_03 [{
   shapeType=Capsule
   length=5
   radius=1.25
   shapeName=dyn_hair_tube_E_03_capsule
   center=<0.2811 0 -0.0397>
   rotation=<rot 86.842 179.9987 179.9987>
  }]
 dyn_hair_tube_E_MSTR [{shapeType=Capsule, length=0.505, radius=0.505, shapeName=dyn_hair_tube_E_MSTR_capsule, rotation=<rot 0 -90 0>}]
 dyn_hair_tube_F_01 [{shapeType=Capsule, length=4, radius=1.25, shapeName=dyn_hair_tube_F_01_capsule, center=<0.4457 0 0>, rotation=<rot 90 0 0>}]
 dyn_hair_tube_F_02 [{shapeType=Capsule, length=4, radius=1.25, shapeName=dyn_hair_tube_F_02_capsule, center=<0.25 0 0>, rotation=<rot 90 0 0>}]
 dyn_hair_tube_F_03 [{shapeType=Capsule, length=4, radius=1.25, shapeName=dyn_hair_tube_F_03_capsule, center=<0.25 0 0>, rotation=<rot 90 0 0>}]
 dyn_hair_tube_F_MSTR [{shapeType=Capsule, length=0.505, radius=0.505, shapeName=dyn_hair_tube_F_MSTR_capsule, rotation=<rot 0 -90 0>}]
 dyn_hair_tube_G_01 [{shapeType=Capsule, length=5, radius=1.25, shapeName=dyn_hair_tube_G_01_capsule, center=<1.7845 0 0>, rotation=<rot 90 0 0>}]
 dyn_hair_tube_G_02 [{
   shapeType=Capsule
   length=4.3245
   radius=1.25
   shapeName=dyn_hair_tube_G_02_capsule
   center=<1 0 0>
   rotation=<rot 80.0709 -179.9999 -179.9999>
  }]
 dyn_hair_tube_G_03 [{
   shapeType=Capsule
   length=4.211
   radius=1.25
   shapeName=dyn_hair_tube_G_03_capsule
   center=<1 0 0>
   rotation=<rot 86.841 -179.9997 -179.9997>
  }]
 dyn_hair_tube_G_04 [{shapeType=Capsule, length=3.1812, radius=1.25, shapeName=dyn_hair_tube_G_04_capsule, center=<1 0 0>, rotation=<rot 90 0 0>}]
 dyn_hair_tube_G_MSTR [{shapeType=Capsule, length=0.505, radius=0.505, shapeName=dyn_hair_tube_G_MSTR_capsule, rotation=<rot 0 -90 0>}]
 dyn_hair_tube_H_01 [{
   shapeType=Capsule
   length=4
   radius=1.25
   shapeName=dyn_hair_tube_H_01_capsule
   center=<1.5542 0 0.0551>
   rotation=<rot 84.3218 179.9993 179.9993>
  }]
 dyn_hair_tube_H_02 [{
   shapeType=Capsule
   length=4
   radius=1.25
   shapeName=dyn_hair_tube_H_02_capsule
   center=<0.7511 0 -0.0595>
   rotation=<rot 74.4338 179.9996 179.9996>
  }]
 dyn_hair_tube_H_03 [{shapeType=Capsule, length=4, radius=1.25, shapeName=dyn_hair_tube_H_03_capsule, center=<0.5718 0 0>, rotation=<rot 90 0 0>}]
 dyn_hair_tube_H_MSTR [{shapeType=Capsule, length=0.505, radius=0.505, shapeName=dyn_hair_tube_H_MSTR_capsule, rotation=<rot 0 -90 0>}]
 dyn_hair_tube_I_01 [{shapeType=Capsule, length=3.4726, radius=1.3497, shapeName=dyn_hair_tube_I_01_capsule, center=<1 0 0>, rotation=<rot 90 0 0>}]
 dyn_hair_tube_I_02 [{shapeType=Capsule, length=3.4726, radius=1.3497, shapeName=dyn_hair_tube_I_02_capsule, center=<1 0 0>, rotation=<rot 87.2657 -180 -179.9999>}]
 dyn_hair_tube_I_03 [{shapeType=Capsule, length=3.4726, radius=1.3497, shapeName=dyn_hair_tube_I_03_capsule, center=<1 0 0>, rotation=<rot 83.895 -0.0001 0>}]
 dyn_hair_tube_I_MSTR [{shapeType=Capsule, length=0.505, radius=0.505, shapeName=dyn_hair_tube_I_MSTR_capsule, rotation=<rot 0 -90 0>}]
 dyn_hair_tube_J_01 [{shapeType=Capsule, length=3.5288, radius=1.25, shapeName=dyn_hair_tube_J_01_capsule, center=<0.3566 0 0>, rotation=<rot 90 -7.08 0>}]
 dyn_hair_tube_J_02 [{
   shapeType=Capsule
   length=3.5288
   radius=1.25
   shapeName=dyn_hair_tube_J_02_capsule
   center=<0.5643 -0.0522 0>
   rotation=<rot 86.7236 165.8995 179.9991>
  }]
 dyn_hair_tube_J_03 [{shapeType=Capsule, length=3.1326, radius=1.25, shapeName=dyn_hair_tube_J_03_capsule, center=<0.2494 0 0>, rotation=<rot 90 0 0>}]
 dyn_hair_tube_J_MSTR [{shapeType=Capsule, length=0.505, radius=0.505, shapeName=dyn_hair_tube_J_MSTR_capsule, rotation=<rot 0 -90 0>}]
 dyn_hair_tube_K_01 [{shapeType=Capsule, length=4.9841, radius=1.3656, shapeName=dyn_hair_tube_K_01_capsule, center=<2.2076 0 0>, rotation=<rot 90 0 0>}]
 dyn_hair_tube_K_02 [{
   shapeType=Capsule
   length=3.8844
   radius=1.3656
   shapeName=dyn_hair_tube_K_02_capsule
   center=<1 0 0>
   rotation=<rot 80.6977 179.9997 179.9998>
  }]
 dyn_hair_tube_K_03 [{
   shapeType=Capsule
   length=3.8844
   radius=1.3656
   shapeName=dyn_hair_tube_K_03_capsule
   center=<1 0 0>
   rotation=<rot 85.2727 -179.9999 -179.9999>
  }]
 dyn_hair_tube_K_04 [{shapeType=Capsule, length=3.8844, radius=1.3656, shapeName=dyn_hair_tube_K_04_capsule, center=<0.6953 0 0>, rotation=<rot 90 0 0>}]
 dyn_hair_tube_K_MSTR [{shapeType=Capsule, length=0.505, radius=0.505, shapeName=dyn_hair_tube_K_MSTR_capsule, rotation=<rot 0 -90 0>}]
 spine_04 [{shapeType=Capsule, length=10.7032, radius=16.7178, shapeName=spine_04_capsule, center=<6.8514 -1.723 0>, rotation=<rot 0 -90 0>}]
 spine_05 [{
   shapeType=Capsule
   length=0.8336
   radius=16.354
   shapeName=spine_05_capsule
   center=<6.9573 0.6445 -0.0001>
   rotation=<rot 0.0025 -90 -0.0017>
  }]
}
constraints [
 {bone1Name=dyn_hair_tube_A_01, bone2Name=dyn_hair_tube_A_MSTR, swing1LimitDegrees=20, swing2LimitDegrees=20, twistLimitDegrees=4}
 {bone1Name=dyn_hair_tube_A_02, bone2Name=dyn_hair_tube_A_01, swing1LimitDegrees=20, swing2LimitDegrees=20, twistLimitDegrees=4}
 {bone1Name=dyn_hair_tube_A_03, bone2Name=dyn_hair_tube_A_02, swing1LimitDegrees=30, swing2LimitDegrees=30, twistLimitDegrees=4}
 {bone1Name=dyn_hair_tube_A_04, bone2Name=dyn_hair_tube_A_03, swing1LimitDegrees=30, swing2LimitDegrees=30, twistLimitDegrees=4}
 {bone1Name=dyn_hair_tube_B_01, bone2Name=dyn_hair_tube_B_MSTR, swing1LimitDegrees=20, swing2LimitDegrees=20, twistLimitDegrees=4}
 {bone1Name=dyn_hair_tube_B_02, bone2Name=dyn_hair_tube_B_01, swing1LimitDegrees=20, swing2LimitDegrees=20, twistLimitDegrees=4}
 {bone1Name=dyn_hair_tube_B_03, bone2Name=dyn_hair_tube_B_02, swing1LimitDegrees=30, swing2LimitDegrees=30, twistLimitDegrees=4}
 {bone1Name=dyn_hair_tube_B_04, bone2Name=dyn_hair_tube_B_03, swing1LimitDegrees=30, swing2LimitDegrees=30, twistLimitDegrees=4}
 {bone1Name=dyn_hair_tube_C_01, bone2Name=dyn_hair_tube_C_MSTR, swing1LimitDegrees=20, swing2LimitDegrees=20, twistLimitDegrees=4}
 {bone1Name=dyn_hair_tube_C_02, bone2Name=dyn_hair_tube_C_01, swing1LimitDegrees=20, swing2LimitDegrees=20, twistLimitDegrees=4}
 {bone1Name=dyn_hair_tube_C_03, bone2Name=dyn_hair_tube_C_02, swing1LimitDegrees=30, swing2LimitDegrees=30, twistLimitDegrees=4}
 {bone1Name=dyn_hair_tube_D_01, bone2Name=dyn_hair_tube_D_MSTR, swing1LimitDegrees=20, swing2LimitDegrees=20, twistLimitDegrees=4}
 {bone1Name=dyn_hair_tube_D_02, bone2Name=dyn_hair_tube_D_01, swing1LimitDegrees=20, swing2LimitDegrees=20, twistLimitDegrees=4}
 {bone1Name=dyn_hair_tube_D_03, bone2Name=dyn_hair_tube_D_02, swing1LimitDegrees=30, swing2LimitDegrees=30, twistLimitDegrees=4}
 {bone1Name=dyn_hair_tube_E_01, bone2Name=dyn_hair_tube_E_MSTR, swing1LimitDegrees=20, swing2LimitDegrees=20, twistLimitDegrees=4}
 {bone1Name=dyn_hair_tube_E_02, bone2Name=dyn_hair_tube_E_01, swing1LimitDegrees=20, swing2LimitDegrees=20, twistLimitDegrees=4}
 {bone1Name=dyn_hair_tube_E_03, bone2Name=dyn_hair_tube_E_02, swing1LimitDegrees=30, swing2LimitDegrees=30, twistLimitDegrees=4}
 {bone1Name=dyn_hair_tube_F_01, bone2Name=dyn_hair_tube_F_MSTR, swing1LimitDegrees=20, swing2LimitDegrees=20, twistLimitDegrees=4}
 {bone1Name=dyn_hair_tube_F_02, bone2Name=dyn_hair_tube_F_01, swing1LimitDegrees=20, swing2LimitDegrees=20, twistLimitDegrees=4}
 {bone1Name=dyn_hair_tube_F_03, bone2Name=dyn_hair_tube_F_02, swing1LimitDegrees=30, swing2LimitDegrees=30, twistLimitDegrees=4}
 {bone1Name=dyn_hair_tube_G_01, bone2Name=dyn_hair_tube_G_MSTR, swing1LimitDegrees=20, swing2LimitDegrees=20, twistLimitDegrees=4}
 {bone1Name=dyn_hair_tube_G_02, bone2Name=dyn_hair_tube_G_01, swing1LimitDegrees=20, swing2LimitDegrees=20, twistLimitDegrees=4}
 {bone1Name=dyn_hair_tube_G_03, bone2Name=dyn_hair_tube_G_02, swing1LimitDegrees=30, swing2LimitDegrees=30, twistLimitDegrees=4}
 {bone1Name=dyn_hair_tube_G_04, bone2Name=dyn_hair_tube_G_03, swing1LimitDegrees=30, swing2LimitDegrees=30, twistLimitDegrees=4}
 {bone1Name=dyn_hair_tube_H_01, bone2Name=dyn_hair_tube_H_MSTR, swing1LimitDegrees=20, swing2LimitDegrees=20, twistLimitDegrees=4}
 {bone1Name=dyn_hair_tube_H_02, bone2Name=dyn_hair_tube_H_01, swing1LimitDegrees=20, swing2LimitDegrees=20, twistLimitDegrees=4}
 {bone1Name=dyn_hair_tube_H_03, bone2Name=dyn_hair_tube_H_02, swing1LimitDegrees=30, swing2LimitDegrees=30, twistLimitDegrees=4}
 {bone1Name=dyn_hair_tube_I_01, bone2Name=dyn_hair_tube_I_MSTR, swing1LimitDegrees=20, swing2LimitDegrees=20, twistLimitDegrees=4}
 {bone1Name=dyn_hair_tube_I_02, bone2Name=dyn_hair_tube_I_01, swing1LimitDegrees=20, swing2LimitDegrees=20, twistLimitDegrees=4}
 {bone1Name=dyn_hair_tube_I_03, bone2Name=dyn_hair_tube_I_02, swing1LimitDegrees=30, swing2LimitDegrees=30, twistLimitDegrees=4}
 {bone1Name=dyn_hair_tube_J_01, bone2Name=dyn_hair_tube_J_MSTR, swing1LimitDegrees=20, swing2LimitDegrees=20, twistLimitDegrees=4}
 {bone1Name=dyn_hair_tube_J_02, bone2Name=dyn_hair_tube_J_01, swing1LimitDegrees=20, swing2LimitDegrees=20, twistLimitDegrees=4}
 {bone1Name=dyn_hair_tube_J_03, bone2Name=dyn_hair_tube_J_02, swing1LimitDegrees=30, swing2LimitDegrees=30, twistLimitDegrees=4}
 {bone1Name=dyn_hair_tube_K_01, bone2Name=dyn_hair_tube_K_MSTR, swing1LimitDegrees=20, swing2LimitDegrees=20, twistLimitDegrees=4}
 {bone1Name=dyn_hair_tube_K_02, bone2Name=dyn_hair_tube_K_01, swing1LimitDegrees=20, swing2LimitDegrees=20, twistLimitDegrees=4}
 {bone1Name=dyn_hair_tube_K_03, bone2Name=dyn_hair_tube_K_02, swing1LimitDegrees=30, swing2LimitDegrees=30, twistLimitDegrees=4}
 {bone1Name=dyn_hair_tube_K_04, bone2Name=dyn_hair_tube_K_03, swing1LimitDegrees=30, swing2LimitDegrees=30, twistLimitDegrees=4}
]
