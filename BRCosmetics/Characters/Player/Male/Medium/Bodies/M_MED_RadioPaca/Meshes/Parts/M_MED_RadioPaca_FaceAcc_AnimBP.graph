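# Post-copy AnimBP animating hair and nose ring on the RadioPaca player cosmetic's face accessory.
# Copied parent pose feeds a harmonics Control Rig, hair-knot AnimDynamics, head-space RigidBody, nose-ring AnimDynamics, then additive hair-bone offsets.
# With rigid body enabled, skydive state or ground speed tunes harmonics speed and RigidBody alpha; skydiving also enables the master hair offset.
# AnimDynAlpha, AnimDynSimSpace and Head_Back_Alpha are set but unread by the AnimGraph.
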AnimBlueprint /BRCosmetics/Characters/Player/Male/Medium/Bodies/M_MED_RadioPaca/Meshes/Parts/M_MED_RadioPaca_FaceAcc_AnimBP
target_skeleton=&/M_MED_RadioPaca_FaceAcc_Skeleton
graphs [
 AnimGraph {nodes [
   n1 = Control Rig!(
    Source=Copy Pose From Mesh() {bUseAttachedParent=true}
    Alpha
    HairHarmonicsSpeed:double=Get HairHarmonicsSpeed() -> (HairHarmonicsSpeed:double)
   ) -> (Pose) {
    alphaBoolBlend {blendOption=Linear}
    controlRigAssetReference {blueprintRigClass=&/M_MED_RadioPaca_FaceAcc_CtrlRig.M_MED_RadioPaca_FaceAcc_CtrlRig_C}
   }
   n2 = Anim Dynamics - Bone: dyn_hair_knot!(SimulationSpace:AnimPhysSimSpaceType Enum=Component, ComponentPose=Local To Component(LocalPose=@n1), Alpha=1, ExternalForce="0,0,0") -> (Pose) {
    angularDampingOverride=1
    angularSpringConstant=50000
    bAngularSpring=true
    bOverrideAngularDamping=true
    bUseGravityOverride=true
    simulationSpace=RootRelative
    alphaBoolBlend {blendOption=Linear}
    boundBone {boneName=dyn_hair_knot}
    gravityOverride {z=-980}
    physicsBodyDefinitions [{
      boundBone {boneName=dyn_hair_knot}
      boxExtents=<60 60 60>
      constraintSetup {angularLimitsMax=<0.5 0.5 0.5>, angularLimitsMin=<-0.5 -0.5 -0.5>, angularTarget=<1 0 0>}
      localJointOffset=<0.05 0 0>
     }]
    relativeSpaceBone {boneName=head}
   }
   n3 = RigidBody(ComponentPose=@n2, Alpha=Get RBAN_Alpha() -> (RBAN_Alpha:double)) {
    bOverrideWorldGravity=true
    overridePhysicsAsset=&/M_MED_RadioPaca_FaceAcc_Physics
    simulationSpace=BaseBoneSpace
    alphaBoolBlend {blendOption=Linear}
    baseBoneRef {boneName=head}
    overrideWorldGravity {z=-980}
    simSpaceSettings {
     dampingAlpha=0
     maxAngularAcceleration=50
     maxAngularVelocity=10
     maxLinearAcceleration=3000
     maxLinearVelocity=600
     worldAlpha=1
     externalLinearDragV=<0.2 0.2 0.2>
    }
   }
   n4 = Anim Dynamics - Bone: dyn_nose_ring!(SimulationSpace:AnimPhysSimSpaceType Enum, ComponentPose=@n3, Alpha=1, ExternalForce="0,0,0") -> (Pose) {
    angularDampingOverride=0.999
    angularSpringConstant=250
    bAngularSpring=true
    bOverrideAngularDamping=true
    bUseGravityOverride=true
    simulationSpace=RootRelative
    alphaBoolBlend {blendOption=Linear}
    boundBone {boneName=dyn_nose_ring}
    gravityOverride {z=-980}
    physicsBodyDefinitions [{
      boundBone {boneName=dyn_nose_ring}
      boxExtents=<5 5 5>
      constraintSetup {angularLimitsMax=<0 10 0>, angularLimitsMin=<0 -45 0>, angularTarget=<1 0 0>}
      localJointOffset=<1 0 0>
     }]
    relativeSpaceBone {boneName=head}
   }
   n5 = "Transform (Modify) Bone - Bone: dyn_hair_tube_B_MSTR"(ComponentPose=@n4, Alpha=1, Translation="0,0,0", Rotation="20,0,0", Scale="1,1,1") {
    rotationMode=BMM_Additive
    rotationSpace=BCS_BoneSpace
    alphaBoolBlend {blendOption=Linear}
    boneToModify {boneName=dyn_hair_tube_B_MSTR}
    rotation {pitch=20}
   }
   n6 = "Transform (Modify) Bone - Bone: dyn_hair_tube_D_MSTR"(ComponentPose=@n5, Alpha=1, Translation="0,0,0", Rotation="20,0,0", Scale="1,1,1") {
    rotationMode=BMM_Additive
    rotationSpace=BCS_BoneSpace
    alphaBoolBlend {blendOption=Linear}
    boneToModify {boneName=dyn_hair_tube_D_MSTR}
    rotation {pitch=20}
   }
   n7 = "Transform (Modify) Bone - Bone: dyn_hair_tube_G_MSTR"(ComponentPose=@n6, Alpha=1, Translation="0,0,0", Rotation="20,0,0", Scale="1,1,1") {
    rotationMode=BMM_Additive
    rotationSpace=BCS_BoneSpace
    alphaBoolBlend {blendOption=Linear}
    boneToModify {boneName=dyn_hair_tube_G_MSTR}
    rotation {pitch=20}
   }
   n8 = "Transform (Modify) Bone - Bone: dyn_hair_tube_K_MSTR"(ComponentPose=@n7, Alpha=1, Translation="0,0,0", Rotation="20,0,0", Scale="1,1,1") {
    rotationMode=BMM_Additive
    rotationSpace=BCS_BoneSpace
    alphaBoolBlend {blendOption=Linear}
    boneToModify {boneName=dyn_hair_tube_K_MSTR}
    rotation {pitch=20}
   }
   n9 = "Transform (Modify) Bone - Bone: dyn_hair_tube_J_MSTR"(ComponentPose=@n8, Alpha=1, Translation="0,0,0", Rotation="20,0,0", Scale="1,1,1") {
    rotationMode=BMM_Additive
    rotationSpace=BCS_BoneSpace
    alphaBoolBlend {blendOption=Linear}
    boneToModify {boneName=dyn_hair_tube_J_MSTR}
    rotation {pitch=20}
   }
   n10 = "Transform (Modify) Bone - Bone: dyn_hair_MSTR"(ComponentPose=@n9, Alpha=Get HairSkydiveAdjust() -> (HairSkydiveAdjust:double), Translation="1,0,0", Rotation="18,0,0", Scale="1,1,1") {
    rotationMode=BMM_Additive
    rotationSpace=BCS_BoneSpace
    translationMode=BMM_Additive
    translationSpace=BCS_BoneSpace
    alphaBoolBlend {blendOption=Linear}
    alphaScaleBiasClamp {bInterpResult=true}
    boneToModify {boneName=dyn_hair_MSTR}
    rotation {pitch=18}
    translation {x=1}
   }
   Output Pose(Result=Component To Local(ComponentPose=@n10)) {layerGroup=None}
  ]}
 EventGraph {nodes [
   Event BlueprintUpdateAnimation()
   > n1 = Sequence!(execute) -> (then_0, then_1, then_2:Exec, then_3:Exec)
   > n2 = Branch(Condition=Get bIsInFrontEnd())
   > Set AnimDynSimSpace(AnimDynSimSpace:AnimPhysSimSpaceType Enum=World) -> (Output_Get:AnimPhysSimSpaceType Enum)
   > Set AnimDynAlpha(AnimDynAlpha:double=0.4) -> (Output_Get:double)
   Set AnimDynSimSpace(execute=@n2.else)
   > Set AnimDynAlpha(AnimDynAlpha=1)
   Set Head_Back_Alpha(execute=@n1.then_1, Head_Back_Alpha:double=GetCurveValue(CurveName=head_back_SPR)) -> (Output_Get:double)
   n3 = Branch(execute=@n1.then_2, Condition=Get bRigidBodyEnabled())
   > n4 = Branch(Condition=Get bIsSkyDiving())
   > Set RBAN_Alpha(RBAN_Alpha:double=0.6) -> (Output_Get:double)
   > Set HairSkydiveAdjust(HairSkydiveAdjust:double=1) -> (Output_Get:double)
   > n5 = Branch(Condition=Get bIsSkydiveFloating())
   > Set HairHarmonicsSpeed(HairHarmonicsSpeed:double=1.5) -> (Output_Get:double)
   n6 = Branch(execute=@n5.else, Condition=Get bIsSkydiveDiveMode())
   > Set HairHarmonicsSpeed(HairHarmonicsSpeed=1.75)
   Set HairHarmonicsSpeed(execute=@n6.else, HairHarmonicsSpeed=1.25)
   Set RBAN_Alpha(execute=@n4.else, RBAN_Alpha=1)
   > Set HairSkydiveAdjust(HairSkydiveAdjust=0)
   n7 = Get PawnSpeed2D()
   > Set HairHarmonicsSpeed(HairHarmonicsSpeed=MapRangeClamped(Value=@n7, InRangeA=250, InRangeB=1000, OutRangeB=1.25))
   > Set RBAN_Alpha(RBAN_Alpha=float - float(A=1, B=MapRangeClamped(Value=@n7, InRangeA=700, InRangeB=1500, OutRangeB=0.6)))
   Set HairHarmonicsSpeed(execute=@n3.else, HairHarmonicsSpeed=0)
  ]}
]
variables [
 AnimDynAlpha {type=number}
 AnimDynSimSpace {type=AnimPhysSimSpaceType, default=RootRelative}
 HairHarmonicsSpeed {type=number}
 HairSkydiveAdjust {type=number}
 Head_Back_Alpha {type=number}
 RBAN_Alpha {type=number}
]
